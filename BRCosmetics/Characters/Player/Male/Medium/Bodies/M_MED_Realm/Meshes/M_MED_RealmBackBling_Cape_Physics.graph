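# Physics asset for the Realm male cape back bling, with capsules for body, cape and coat bones plus a large sphere centred below root.
# Four six-segment cape chains hang from master bodies on clavicles and upper spine, using tight swing limits and zero-degree twist.
# Coat chains attach to the pelvis with free angular motion, as do body joints.
# Lateral constraints tie neighbouring cape and coat chains together, mostly angularly free.
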PhysicsAsset /BRCosmetics/Characters/Player/Male/Medium/Bodies/M_MED_Realm/Meshes/M_MED_RealmBackBling_Cape_Physics
bodies {
 calf_l [{shapeType=Capsule, length=43.179, radius=7.1561, shapeName=calf_l_capsule, center=<-18.4993 0 0>, rotation=<rot 90 0 0>}]
 calf_r [{shapeType=Capsule, length=43.179, radius=7.1561, shapeName=calf_r_capsule, center=<18.4993 0 0>, rotation=<rot 90 0 0>}]
 cape_01_l [{shapeType=Capsule, length=15.0197, radius=4.9509, shapeName=cape_01_l_capsule, center=<8.8684 0 0>, rotation=<rot 90 0 90>}]
 cape_01_r [{shapeType=Capsule, length=9.6914, radius=5.0845, shapeName=cape_01_r_capsule, center=<8.2303 0 0>, rotation=<rot 90 0 0>}]
 cape_02_l [{shapeType=Capsule, length=21.5299, radius=8.525, shapeName=cape_02_l_capsule, center=<28.2058 0 0>, rotation=<rot -0.0002 0 100>}]
 cape_02_r [{shapeType=Capsule, length=24.0071, radius=8.3933, shapeName=cape_02_r_capsule, center=<27.3308 0 0>, rotation=<rot -0.0003 0 85>}]
 cape_03_l [{shapeType=Capsule, length=11.8527, radius=8.525, shapeName=cape_03_l_capsule, center=<17.6378 0 0>, rotation=<rot -0.0002 0 100>}]
 cape_03_r [{shapeType=Capsule, length=14.2709, radius=9.5429, shapeName=cape_03_r_capsule, center=<14.5802 -5.7118 -0.4997>, rotation=<rot -0.0002 0 85>}]
 cape_04_l [{shapeType=Capsule, length=15.1765, radius=9, shapeName=cape_04_l_capsule, center=<10.1502 0 0>, rotation=<rot -0.0002 0 -95>}]
 cape_04_r [{shapeType=Capsule, length=15.6761, radius=9, shapeName=cape_04_r_capsule, center=<10.1502 0 0>, rotation=<rot -0.0002 0 -95>}]
 cape_05_l [{shapeType=Capsule, length=17.9398, radius=9, shapeName=cape_05_l_capsule, center=<7.655 0 0>, rotation=<rot 90 0 90>}]
 cape_05_r [{shapeType=Capsule, length=16.5039, radius=9, shapeName=cape_05_r_capsule, center=<8 -10 0>, rotation=<rot 0 0 90>}]
 cape_06_l [{shapeType=Capsule, length=17.9874, radius=9, shapeName=cape_06_l_capsule, center=<7.655 0 0>, rotation=<rot 0 0 90>}]
 cape_06_r [{shapeType=Capsule, length=14.0923, radius=9, shapeName=cape_06_r_capsule, center=<8 -10 1.2942>, rotation=<rot 0 0 90>}]
 cape_l_MSTR [{shapeType=Capsule, length=0.5, radius=1.5, shapeName=cape_l_MSTR_capsule, rotation=<rot 0 -90 0>}]
 cape_mid_01_l [{shapeType=Capsule, length=15.772, radius=4.5132, shapeName=cape_mid_01_l_capsule, center=<10.9091 0 5.8>, rotation=<rot 82.4316 -180 -180>}]
 cape_mid_01_r [{shapeType=Capsule, length=15.772, radius=4.5132, shapeName=cape_mid_01_r_capsule, center=<10.9091 0 5.8>, rotation=<rot 82.4316 180 180>}]
 cape_mid_02_l [{shapeType=Capsule, length=15.2976, radius=8.525, shapeName=cape_mid_02_l_capsule, center=<28.2058 0 0>, rotation=<rot -0.0002 0 100>}]
 cape_mid_02_r [{shapeType=Capsule, length=15.2976, radius=8.525, shapeName=cape_mid_02_r_capsule, center=<28.2058 0 0>, rotation=<rot -0.0003 0 85>}]
 cape_mid_03_l [{shapeType=Capsule, length=17.9079, radius=8.525, shapeName=cape_mid_03_l_capsule, center=<17.6378 0 0>, rotation=<rot -0.0002 0 100>}]
 cape_mid_03_r [{shapeType=Capsule, length=17.9079, radius=8.525, shapeName=cape_mid_03_r_capsule, center=<17.6378 0 0>, rotation=<rot -0.0002 0 100>}]
 cape_mid_04_l [{shapeType=Capsule, length=18.0186, radius=9, shapeName=cape_mid_04_l_capsule, center=<10.1502 0 0>, rotation=<rot -0.0002 0 -95>}]
 cape_mid_04_r [{shapeType=Capsule, length=17.1771, radius=9, shapeName=cape_mid_04_r_capsule, center=<10.1502 0 0>, rotation=<rot -0.0002 0 -95>}]
 cape_mid_05_l [{shapeType=Capsule, length=16.9594, radius=9, shapeName=cape_mid_05_l_capsule, center=<7.655 0 0>, rotation=<rot 90 0 90>}]
 cape_mid_05_r [{shapeType=Capsule, length=16.9594, radius=9, shapeName=cape_mid_05_r_capsule, center=<7.655 0 0>, rotation=<rot 90 0 90>}]
 cape_mid_06_l [{shapeType=Capsule, length=17.9874, radius=9, shapeName=cape_mid_06_l_capsule, center=<7.655 0 0>, rotation=<rot 0 0 90>}]
 cape_mid_06_r [{shapeType=Capsule, length=17.9874, radius=9, shapeName=cape_mid_06_r_capsule, center=<7.655 0 0>, rotation=<rot 0 0 90>}]
 cape_mid_l_MSTR [{shapeType=Capsule, length=0.5, radius=1.5, shapeName=cape_mid_l_MSTR_capsule, rotation=<rot 0 -90 0>}]
 cape_mid_r_MSTR [{shapeType=Capsule, length=0.5, radius=1.5, shapeName=cape_mid_r_MSTR_capsule, rotation=<rot 0 -90 0>}]
 cape_r_MSTR [{shapeType=Capsule, length=0.5, radius=1.5, shapeName=cape_r_MSTR_capsule, rotation=<rot 0 -90 0>}]
 clavicle_l [{shapeType=Capsule, length=12.0737, radius=7.5263, shapeName=clavicle_l_capsule, center=<9.3 0 0>, rotation=<rot 81 0 0>}]
 clavicle_r [{shapeType=Capsule, length=12.0737, radius=7.5263, shapeName=clavicle_r_capsule, center=<-9.3 0 0>, rotation=<rot 81 0 0>}]
 dyn_coat_bk_c_02 [{shapeType=Capsule, length=20.5715, radius=3.9997, shapeName=dyn_coat_bk_c_02_capsule, rotation=<rot 90 0 0>}]
 dyn_coat_bk_c_04 [{shapeType=Capsule, length=16.9533, radius=7.6178, shapeName=dyn_coat_bk_c_04_capsule, center=<0 1.8663 0>, rotation=<rot 90 0 0>}]
 dyn_coat_bk_c_06 [{shapeType=Capsule, length=21.0165, radius=9.5546, shapeName=dyn_coat_bk_c_06_capsule, rotation=<rot 90 0 0>}]
 dyn_coat_bk_l_02 [{shapeType=Capsule, length=20.5715, radius=3.9997, shapeName=dyn_coat_bk_l_02_capsule, rotation=<rot 90 0 0>}]
 dyn_coat_bk_l_04 [{shapeType=Capsule, length=18.4533, radius=6.1178, shapeName=dyn_coat_bk_l_04_capsule, rotation=<rot 90 0 0>}]
 dyn_coat_bk_l_06 [{shapeType=Capsule, length=21.0165, radius=9.5546, shapeName=dyn_coat_bk_l_06_capsule, rotation=<rot 90 0 0>}]
 dyn_coat_bk_r_02 [{shapeType=Capsule, length=20.5715, radius=3.9997, shapeName=dyn_coat_bk_r_02_capsule, rotation=<rot 90 0 0>}]
 dyn_coat_bk_r_04 [{shapeType=Capsule, length=17.4533, radius=7.1178, shapeName=dyn_coat_bk_r_04_capsule, center=<0 1.5882 0>, rotation=<rot 90 0 0>}]
 dyn_coat_bk_r_06 [{shapeType=Capsule, length=15.0165, radius=9.5546, shapeName=dyn_coat_bk_r_06_capsule, rotation=<rot 90 0 0>}]
 dyn_coat_l_04 [{shapeType=Capsule, length=16.398, radius=8.1732, shapeName=dyn_coat_l_04_capsule, rotation=<rot 90 0 0>}]
 dyn_coat_l_06 [{shapeType=Capsule, length=22.398, radius=10.1732, shapeName=dyn_coat_l_06_capsule, rotation=<rot 90 0 0>}]
 dyn_coat_r_04 [{shapeType=Capsule, length=15.898, radius=8.6732, shapeName=dyn_coat_r_04_capsule, rotation=<rot 90 0 0>}]
 dyn_coat_r_06 [{shapeType=Capsule, length=21.898, radius=10.6732, shapeName=dyn_coat_r_06_capsule, center=<0 0 -1.0874>, rotation=<rot 90 0 0>}]
 foot_l [{shapeType=Capsule, length=14.3017, radius=10.5, shapeName=foot_l_capsule, center=<-4.0851 7.2082 0>, rotation=<rot 90 180 -90>}]
 foot_r [{shapeType=Capsule, length=12.8613, radius=10.5, shapeName=foot_r_capsule, center=<5.1029 -7.0735 0>, rotation=<rot 90 180 -90>}]
 lowerarm_l [{shapeType=Capsule, length=24.1831, radius=5.6284, shapeName=lowerarm_l_capsule, center=<10.6887 0 0>, rotation=<rot 90 0 0>}]
 lowerarm_r [{shapeType=Capsule, length=18.7314, radius=5.6284, shapeName=lowerarm_r_capsule, center=<-18.4993 0 0>, rotation=<rot 90 0 0>}]
 pelvis [{shapeType=Capsule, length=11.546, radius=12.8138, shapeName=pelvis_capsule}]
 root [{shapeType=Sphere, radius=97.5, shapeName=root_sphere, center=<0 0 -79.9285>}]
 spine_02 [{shapeType=Capsule, length=14.1596, radius=10.2002, shapeName=spine_02_capsule}]
 spine_03 [{shapeType=Capsule, length=13.5763, radius=11.3042, shapeName=spine_03_capsule}]
 spine_04 [{shapeType=Capsule, length=13.7137, radius=12.1257, shapeName=spine_04_capsule}]
 spine_05 [{shapeType=Capsule, length=21.5542, radius=9.794, shapeName=spine_05_capsule}]
 thigh_l [{shapeType=Capsule, length=33.1221, radius=8.1367, shapeName=thigh_l_capsule, center=<-18.4993 0 0>, rotation=<rot 90 0 0>}]
 thigh_r [{shapeType=Capsule, length=33.1221, radius=8.1367, shapeName=thigh_r_capsule, center=<18.4993 0 0>, rotation=<rot 90 0 0>}]
 upperarm_l [{shapeType=Capsule, length=18.3771, radius=5.9826, shapeName=upperarm_l_capsule, center=<14.4803 0 0>, rotation=<rot 90 0 0>}]
 upperarm_r [{shapeType=Capsule, length=17.6862, radius=6.6736, shapeName=upperarm_r_capsule, center=<-18.4993 0 0>, rotation=<rot 90 0 0>}]
}
constraints [
 {
  bone1Name=spine_05
  bone2Name=spine_04
  swing1LimitDegrees=45
  swing1Motion=Free
  swing2LimitDegrees=45
  swing2Motion=Free
  twistLimitDegrees=45
  twistMotion=Free
 }
 {
  bone1Name=clavicle_l
  bone2Name=spine_05
  swing1LimitDegrees=45
  swing1Motion=Free
  swing2LimitDegrees=45
  swing2Motion=Free
  twistLimitDegrees=45
  twistMotion=Free
 }
 {
  bone1Name=spine_04
  bone2Name=spine_03
  swing1LimitDegrees=45
  swing1Motion=Free
  swing2LimitDegrees=45
  swing2Motion=Free
  twistLimitDegrees=45
  twistMotion=Free
 }
 {
  bone1Name=upperarm_l
  bone2Name=clavicle_l
  swing1LimitDegrees=45
  swing1Motion=Free
  swing2LimitDegrees=45
  swing2Motion=Free
  twistLimitDegrees=45
  twistMotion=Free
 }
 {
  bone1Name=spine_02
  bone2Name=pelvis
  swing1LimitDegrees=45
  swing1Motion=Free
  swing2LimitDegrees=45
  swing2Motion=Free
  twistLimitDegrees=45
  twistMotion=Free
 }
 {
  bone1Name=lowerarm_l
  bone2Name=upperarm_l
  swing1LimitDegrees=45
  swing1Motion=Free
  swing2LimitDegrees=45
  swing2Motion=Free
  twistLimitDegrees=45
  twistMotion=Free
 }
 {
  bone1Name=clavicle_r
  bone2Name=spine_05
  swing1LimitDegrees=45
  swing1Motion=Free
  swing2LimitDegrees=45
  swing2Motion=Free
  twistLimitDegrees=45
  twistMotion=Free
 }
 {
  bone1Name=upperarm_r
  bone2Name=clavicle_r
  swing1LimitDegrees=45
  swing1Motion=Free
  swing2LimitDegrees=45
  swing2Motion=Free
  twistLimitDegrees=45
  twistMotion=Free
 }
 {
  bone1Name=lowerarm_r
  bone2Name=upperarm_r
  swing1LimitDegrees=45
  swing1Motion=Free
  swing2LimitDegrees=45
  swing2Motion=Free
  twistLimitDegrees=45
  twistMotion=Free
 }
 {
  bone1Name=thigh_l
  bone2Name=pelvis
  swing1LimitDegrees=45
  swing1Motion=Free
  swing2LimitDegrees=45
  swing2Motion=Free
  twistLimitDegrees=45
  twistMotion=Free
 }
 {
  bone1Name=calf_l
  bone2Name=thigh_l
  swing1LimitDegrees=45
  swing1Motion=Free
  swing2LimitDegrees=45
  swing2Motion=Free
  twistLimitDegrees=45
  twistMotion=Free
 }
 {
  bone1Name=thigh_r
  bone2Name=pelvis
  swing1LimitDegrees=45
  swing1Motion=Free
  swing2LimitDegrees=45
  swing2Motion=Free
  twistLimitDegrees=45
  twistMotion=Free
 }
 {
  bone1Name=calf_r
  bone2Name=thigh_r
  swing1LimitDegrees=45
  swing1Motion=Free
  swing2LimitDegrees=45
  swing2Motion=Free
  twistLimitDegrees=45
  twistMotion=Free
 }
 {
  bone1Name=spine_03
  bone2Name=spine_02
  swing1LimitDegrees=45
  swing1Motion=Free
  swing2LimitDegrees=45
  swing2Motion=Free
  twistLimitDegrees=45
  twistMotion=Free
 }
 {
  bone1Name=cape_l_MSTR
  bone2Name=clavicle_l
  swing1LimitDegrees=45
  swing1Motion=Free
  swing2LimitDegrees=45
  swing2Motion=Free
  twistLimitDegrees=45
  twistMotion=Free
 }
 {bone1Name=cape_01_l, bone2Name=cape_l_MSTR, swing1LimitDegrees=20, swing2LimitDegrees=20, twistLimitDegrees=0}
 {bone1Name=cape_02_l, bone2Name=cape_01_l, swing1LimitDegrees=20, swing2LimitDegrees=20, twistLimitDegrees=0}
 {bone1Name=cape_03_l, bone2Name=cape_02_l, swing1LimitDegrees=15, swing2LimitDegrees=20, twistLimitDegrees=0}
 {bone1Name=cape_04_l, bone2Name=cape_03_l, swing1LimitDegrees=15, swing2LimitDegrees=20, twistLimitDegrees=0}
 {bone1Name=cape_05_l, bone2Name=cape_04_l, swing1LimitDegrees=15, swing2LimitDegrees=20, twistLimitDegrees=0}
 {bone1Name=cape_06_l, bone2Name=cape_05_l, swing1LimitDegrees=20, swing2LimitDegrees=20, twistLimitDegrees=0}
 {
  bone1Name=cape_r_MSTR
  bone2Name=clavicle_r
  swing1LimitDegrees=45
  swing1Motion=Free
  swing2LimitDegrees=45
  swing2Motion=Free
  twistLimitDegrees=45
  twistMotion=Free
 }
 {
  bone1Name=cape_mid_l_MSTR
  bone2Name=spine_05
  swing1LimitDegrees=45
  swing1Motion=Free
  swing2LimitDegrees=45
  swing2Motion=Free
  twistLimitDegrees=45
  twistMotion=Free
 }
 {bone1Name=cape_mid_01_l, bone2Name=cape_mid_l_MSTR, swing1LimitDegrees=5, swing2LimitDegrees=20, twistLimitDegrees=0}
 {bone1Name=cape_mid_02_l, bone2Name=cape_mid_01_l, swing1LimitDegrees=10, swing2LimitDegrees=22, twistLimitDegrees=0}
 {bone1Name=cape_mid_03_l, bone2Name=cape_mid_02_l, swing1LimitDegrees=10, swing2LimitDegrees=24, twistLimitDegrees=0}
 {bone1Name=cape_mid_04_l, bone2Name=cape_mid_03_l, swing1LimitDegrees=10, swing2LimitDegrees=30, twistLimitDegrees=0}
 {bone1Name=cape_mid_05_l, bone2Name=cape_mid_04_l, swing1LimitDegrees=20, swing2LimitDegrees=20, twistLimitDegrees=0}
 {bone1Name=cape_mid_06_l, bone2Name=cape_mid_05_l, swing1LimitDegrees=20, swing2LimitDegrees=20, twistLimitDegrees=0}
 {
  bone1Name=cape_mid_r_MSTR
  bone2Name=spine_05
  swing1LimitDegrees=45
  swing1Motion=Free
  swing2LimitDegrees=45
  swing2Motion=Free
  twistLimitDegrees=45
  twistMotion=Free
 }
 {bone1Name=cape_mid_01_r, bone2Name=cape_mid_r_MSTR, swing1LimitDegrees=5, swing2LimitDegrees=20, twistLimitDegrees=0}
 {bone1Name=cape_mid_02_r, bone2Name=cape_mid_01_r, swing1LimitDegrees=10, swing2LimitDegrees=22, twistLimitDegrees=0}
 {bone1Name=cape_mid_03_r, bone2Name=cape_mid_02_r, swing1LimitDegrees=10, swing2LimitDegrees=24, twistLimitDegrees=0}
 {bone1Name=cape_mid_04_r, bone2Name=cape_mid_03_r, swing1LimitDegrees=10, swing2LimitDegrees=30, twistLimitDegrees=0}
 {bone1Name=cape_mid_05_r, bone2Name=cape_mid_04_r, swing1LimitDegrees=20, swing2LimitDegrees=20, twistLimitDegrees=0}
 {bone1Name=cape_mid_06_r, bone2Name=cape_mid_05_r, swing1LimitDegrees=20, swing2LimitDegrees=20, twistLimitDegrees=0}
 {
  bone1Name=foot_l
  bone2Name=calf_l
  swing1LimitDegrees=45
  swing1Motion=Free
  swing2LimitDegrees=45
  swing2Motion=Free
  twistLimitDegrees=45
  twistMotion=Free
 }
 {
  bone1Name=foot_r
  bone2Name=calf_r
  swing1LimitDegrees=45
  swing1Motion=Free
  swing2LimitDegrees=45
  swing2Motion=Free
  twistLimitDegrees=45
  twistMotion=Free
 }
 {
  bone1Name=dyn_coat_l_04
  bone2Name=pelvis
  swing1LimitDegrees=45
  swing1Motion=Free
  swing2LimitDegrees=45
  swing2Motion=Free
  twistLimitDegrees=45
  twistMotion=Free
 }
 {
  bone1Name=dyn_coat_l_06
  bone2Name=dyn_coat_l_04
  swing1LimitDegrees=45
  swing1Motion=Free
  swing2LimitDegrees=45
  swing2Motion=Free
  twistLimitDegrees=45
  twistMotion=Free
 }
 {
  bone1Name=dyn_coat_bk_l_02
  bone2Name=pelvis
  swing1LimitDegrees=45
  swing1Motion=Free
  swing2LimitDegrees=45
  swing2Motion=Free
  twistLimitDegrees=45
  twistMotion=Free
 }
 {
  bone1Name=dyn_coat_bk_l_04
  bone2Name=dyn_coat_bk_l_02
  swing1LimitDegrees=45
  swing1Motion=Free
  swing2LimitDegrees=45
  swing2Motion=Free
  twistLimitDegrees=45
  twistMotion=Free
 }
 {
  bone1Name=dyn_coat_bk_l_06
  bone2Name=dyn_coat_bk_l_04
  swing1LimitDegrees=45
  swing1Motion=Free
  swing2LimitDegrees=45
  swing2Motion=Free
  twistLimitDegrees=45
  twistMotion=Free
 }
 {
  bone1Name=dyn_coat_bk_c_02
  bone2Name=pelvis
  swing1LimitDegrees=45
  swing1Motion=Free
  swing2LimitDegrees=45
  swing2Motion=Free
  twistLimitDegrees=45
  twistMotion=Free
 }
 {
  bone1Name=dyn_coat_bk_c_04
  bone2Name=dyn_coat_bk_c_02
  swing1LimitDegrees=45
  swing1Motion=Free
  swing2LimitDegrees=45
  swing2Motion=Free
  twistLimitDegrees=45
  twistMotion=Free
 }
 {
  bone1Name=dyn_coat_bk_c_06
  bone2Name=dyn_coat_bk_c_04
  swing1LimitDegrees=45
  swing1Motion=Free
  swing2LimitDegrees=45
  swing2Motion=Free
  twistLimitDegrees=45
  twistMotion=Free
 }
 {
  bone1Name=dyn_coat_r_04
  bone2Name=pelvis
  swing1LimitDegrees=45
  swing1Motion=Free
  swing2LimitDegrees=45
  swing2Motion=Free
  twistLimitDegrees=45
  twistMotion=Free
 }
 {
  bone1Name=dyn_coat_r_06
  bone2Name=dyn_coat_r_04
  swing1LimitDegrees=45
  swing1Motion=Free
  swing2LimitDegrees=45
  swing2Motion=Free
  twistLimitDegrees=45
  twistMotion=Free
 }
 {
  bone1Name=dyn_coat_bk_r_02
  bone2Name=pelvis
  swing1LimitDegrees=45
  swing1Motion=Free
  swing2LimitDegrees=45
  swing2Motion=Free
  twistLimitDegrees=45
  twistMotion=Free
 }
 {
  bone1Name=dyn_coat_bk_r_04
  bone2Name=dyn_coat_bk_r_02
  swing1LimitDegrees=45
  swing1Motion=Free
  swing2LimitDegrees=45
  swing2Motion=Free
  twistLimitDegrees=45
  twistMotion=Free
 }
 {
  bone1Name=dyn_coat_bk_r_06
  bone2Name=dyn_coat_bk_r_04
  swing1LimitDegrees=45
  swing1Motion=Free
  swing2LimitDegrees=45
  swing2Motion=Free
  twistLimitDegrees=45
  twistMotion=Free
 }
 {bone1Name=cape_mid_02_r, bone2Name=cape_02_l, swing1LimitDegrees=20, swing2LimitDegrees=2, twistLimitDegrees=0}
 {bone1Name=cape_mid_02_l, bone2Name=cape_02_l, swing1LimitDegrees=20, swing2LimitDegrees=2, twistLimitDegrees=0}
 {bone1Name=cape_01_r, bone2Name=cape_r_MSTR, swing1LimitDegrees=10, swing2LimitDegrees=20, twistLimitDegrees=0}
 {bone1Name=cape_02_r, bone2Name=cape_01_r, swing1LimitDegrees=10, swing2LimitDegrees=20, twistLimitDegrees=0}
 {bone1Name=cape_03_r, bone2Name=cape_02_r, swing1LimitDegrees=10, swing2LimitDegrees=20, twistLimitDegrees=0}
 {bone1Name=cape_04_r, bone2Name=cape_03_r, swing1LimitDegrees=20, swing2LimitDegrees=20, twistLimitDegrees=0}
 {bone1Name=cape_05_r, bone2Name=cape_04_r, swing1LimitDegrees=20, swing2LimitDegrees=20, twistLimitDegrees=0}
 {bone1Name=cape_06_r, bone2Name=cape_05_r, swing1LimitDegrees=20, swing2LimitDegrees=20, twistLimitDegrees=0}
 {
  bone1Name=cape_mid_06_l
  bone2Name=cape_mid_06_r
  swing1LimitDegrees=45
  swing1Motion=Free
  swing2LimitDegrees=45
  swing2Motion=Free
  twistLimitDegrees=45
  twistMotion=Free
 }
 {
  bone1Name=cape_06_l
  bone2Name=cape_mid_06_l
  swing1LimitDegrees=45
  swing1Motion=Free
  swing2LimitDegrees=45
  swing2Motion=Free
  twistLimitDegrees=45
  twistMotion=Free
 }
 {
  bone1Name=cape_mid_06_r
  bone2Name=cape_06_r
  swing1LimitDegrees=45
  swing1Motion=Free
  swing2LimitDegrees=45
  swing2Motion=Free
  twistLimitDegrees=45
  twistMotion=Free
 }
 {
  bone1Name=cape_mid_06_r
  bone2Name=cape_06_r
  swing1LimitDegrees=45
  swing1Motion=Free
  swing2LimitDegrees=45
  swing2Motion=Free
  twistLimitDegrees=45
  twistMotion=Free
 }
 {
  bone1Name=cape_05_l
  bone2Name=cape_mid_05_l
  swing1LimitDegrees=45
  swing1Motion=Free
  swing2LimitDegrees=45
  swing2Motion=Free
  twistLimitDegrees=45
  twistMotion=Free
 }
 {
  bone1Name=cape_04_l
  bone2Name=cape_mid_04_l
  swing1LimitDegrees=45
  swing1Motion=Free
  swing2LimitDegrees=45
  swing2Motion=Free
  twistLimitDegrees=45
  twistMotion=Free
 }
 {
  bone1Name=cape_03_l
  bone2Name=cape_mid_03_l
  swing1LimitDegrees=45
  swing1Motion=Free
  swing2LimitDegrees=45
  swing2Motion=Free
  twistLimitDegrees=45
  twistMotion=Free
 }
 {
  bone1Name=cape_05_r
  bone2Name=cape_mid_05_r
  swing1LimitDegrees=45
  swing1Motion=Free
  swing2LimitDegrees=45
  swing2Motion=Free
  twistLimitDegrees=45
  twistMotion=Free
 }
 {
  bone1Name=cape_04_r
  bone2Name=cape_mid_04_r
  swing1LimitDegrees=45
  swing1Motion=Free
  swing2LimitDegrees=45
  swing2Motion=Free
  twistLimitDegrees=45
  twistMotion=Free
 }
 {
  bone1Name=cape_03_r
  bone2Name=cape_mid_03_r
  swing1LimitDegrees=45
  swing1Motion=Free
  swing2LimitDegrees=45
  swing2Motion=Free
  twistLimitDegrees=45
  twistMotion=Free
 }
 {
  bone1Name=dyn_coat_r_04
  bone2Name=dyn_coat_bk_r_04
  swing1LimitDegrees=45
  swing1Motion=Free
  swing2LimitDegrees=45
  swing2Motion=Free
  twistLimitDegrees=45
  twistMotion=Free
 }
 {
  bone1Name=dyn_coat_r_06
  bone2Name=dyn_coat_bk_r_06
  swing1LimitDegrees=45
  swing1Motion=Free
  swing2LimitDegrees=45
  swing2Motion=Free
  twistLimitDegrees=45
  twistMotion=Free
 }
 {
  bone1Name=dyn_coat_l_06
  bone2Name=dyn_coat_bk_l_06
  swing1LimitDegrees=45
  swing1Motion=Free
  swing2LimitDegrees=45
  swing2Motion=Free
  twistLimitDegrees=45
  twistMotion=Free
 }
 {
  bone1Name=dyn_coat_l_04
  bone2Name=dyn_coat_bk_l_04
  swing1LimitDegrees=45
  swing1Motion=Free
  swing2LimitDegrees=45
  swing2Motion=Free
  twistLimitDegrees=45
  twistMotion=Free
 }
 {
  bone1Name=dyn_coat_bk_r_06
  bone2Name=dyn_coat_bk_c_06
  swing1LimitDegrees=45
  swing1Motion=Free
  swing2LimitDegrees=45
  swing2Motion=Free
  twistLimitDegrees=45
  twistMotion=Free
 }
 {
  bone1Name=dyn_coat_bk_l_06
  bone2Name=dyn_coat_bk_c_06
  swing1LimitDegrees=45
  swing1Motion=Free
  swing2LimitDegrees=45
  swing2Motion=Free
  twistLimitDegrees=45
  twistMotion=Free
 }
 {
  bone1Name=pelvis
  bone2Name=root
  swing1LimitDegrees=45
  swing1Motion=Free
  swing2LimitDegrees=45
  swing2Motion=Free
  twistLimitDegrees=45
  twistMotion=Free
 }
]
