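# Animation Blueprint for the male medium Realm NPC cosmetic body, adding deformation and secondary motion to its coat, cape and lightsaber bones.
# The pose copied from the parent mesh runs through two deform Control Rigs into cached poses.
# Cape RigidBody sims, chosen by backbling, blend against a crouch pose using hand-offset coat bones.
# Post-sim, additive cape offsets follow sprint, skydive and parachute states, while constraints and AnimDynamics drive the lightsaber.
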AnimBlueprint /BRCosmetics/Characters/Player/Male/Medium/Bodies/M_MED_Realm_NPC/Meshes/M_MED_Realm_NPC_AnimBP
target_skeleton=&/M_MED_Realm_NPC_Skeleton
graphs [
 AnimGraph {nodes [
   n1 = "Transform (Modify) Bone - Bone: dyn_coat_front_l_01"(Alpha=0, Translation="0,0,0", Rotation="14.442,-9.634,-2.425", Scale="1,1,1") {
    rotationMode=BMM_Additive
    rotationSpace=BCS_BoneSpace
    translationMode=BMM_Additive
    translationSpace=BCS_BoneSpace
    alphaBoolBlend {blendOption=Linear}
    boneToModify {boneName=dyn_coat_front_l_01}
    rotation=<rot 14.4421 -9.634 -2.425>
   }
   n2 = "Transform (Modify) Bone - Bone: dyn_coat_front_l_02"(ComponentPose=@n1, Alpha=0, Translation="0,0,0", Rotation="-0.001,-4.831,0.005", Scale="1,1,1") {
    rotationMode=BMM_Additive
    rotationSpace=BCS_BoneSpace
    translationMode=BMM_Additive
    translationSpace=BCS_BoneSpace
    alphaBoolBlend {blendOption=Linear}
    boneToModify {boneName=dyn_coat_front_l_02}
    rotation=<rot -0.001 -4.831 0.005>
   }
   n3 = "Transform (Modify) Bone - Bone: dyn_coat_front_l_03"(ComponentPose=@n2, Alpha=0, Translation="0,0,0", Rotation="-4.621,-0.002,-0.000", Scale="1,1,1") {
    rotationMode=BMM_Additive
    rotationSpace=BCS_BoneSpace
    translationMode=BMM_Additive
    translationSpace=BCS_BoneSpace
    alphaBoolBlend {blendOption=Linear}
    boneToModify {boneName=dyn_coat_front_l_03}
    rotation {pitch=-4.621, yaw=-0.002}
   }
   n4 = "Transform (Modify) Bone - Bone: dyn_coat_l_01"(ComponentPose=@n3, Alpha=0, Translation="0,0,0", Rotation="0,0,0", Scale="1,1,1") {
    rotationMode=BMM_Additive
    rotationSpace=BCS_BoneSpace
    translationMode=BMM_Additive
    translationSpace=BCS_BoneSpace
    alphaBoolBlend {blendOption=Linear}
    boneToModify {boneName=dyn_coat_l_01}
   }
   n5 = "Transform (Modify) Bone - Bone: dyn_coat_l_02"(ComponentPose=@n4, Alpha=0, Translation="0,0,0", Rotation="0,0,0", Scale="1,1,1") {
    rotationMode=BMM_Additive
    rotationSpace=BCS_BoneSpace
    translationMode=BMM_Additive
    translationSpace=BCS_BoneSpace
    alphaBoolBlend {blendOption=Linear}
    boneToModify {boneName=dyn_coat_l_02}
   }
   n6 = "Transform (Modify) Bone - Bone: dyn_coat_l_03"(ComponentPose=@n5, Alpha=0, Translation="0,0,0", Rotation="0,0,0", Scale="1,1,1") {
    rotationMode=BMM_Additive
    rotationSpace=BCS_BoneSpace
    translationMode=BMM_Additive
    translationSpace=BCS_BoneSpace
    alphaBoolBlend {blendOption=Linear}
    boneToModify {boneName=dyn_coat_l_03}
   }
   n7 = "Transform (Modify) Bone - Bone: dyn_coat_front_r_01"(ComponentPose=@n6, Alpha=1, Translation="0,0,0", Rotation="0.021,-7.765,-12.533", Scale="1,1,1") {
    rotationMode=BMM_Additive
    rotationSpace=BCS_BoneSpace
    translationMode=BMM_Additive
    translationSpace=BCS_BoneSpace
    alphaBoolBlend {blendOption=Linear}
    boneToModify {boneName=dyn_coat_front_r_01}
    rotation=<rot 0.021 -7.765 -12.533>
   }
   n8 = "Transform (Modify) Bone - Bone: dyn_coat_front_r_02"(ComponentPose=@n7, Alpha=1, Translation="0,0,0", Rotation="0.740,-4.473,-0.000", Scale="1,1,1") {
    rotationMode=BMM_Additive
    rotationSpace=BCS_BoneSpace
    translationMode=BMM_Additive
    translationSpace=BCS_BoneSpace
    alphaBoolBlend {blendOption=Linear}
    boneToModify {boneName=dyn_coat_front_r_02}
    rotation=<rot 0.74 -4.473 0>
   }
   n9 = "Transform (Modify) Bone - Bone: dyn_coat_front_r_03"(ComponentPose=@n8, Alpha=1, Translation="0,0,0", Rotation="0,0,0", Scale="1,1,1") {
    rotationMode=BMM_Additive
    rotationSpace=BCS_BoneSpace
    translationMode=BMM_Additive
    translationSpace=BCS_BoneSpace
    alphaBoolBlend {blendOption=Linear}
    boneToModify {boneName=dyn_coat_front_r_03}
   }
   n10 = "Transform (Modify) Bone - Bone: dyn_coat_r_01"(ComponentPose=@n9, Alpha=1, Translation="0,0,0", Rotation="2.000,-4.128,-0.000", Scale="1,1,1") {
    rotationMode=BMM_Additive
    rotationSpace=BCS_BoneSpace
    translationMode=BMM_Additive
    translationSpace=BCS_BoneSpace
    alphaBoolBlend {blendOption=Linear}
    boneToModify {boneName=dyn_coat_r_01}
    rotation=<rot 2 -4.128 0>
   }
   n11 = "Transform (Modify) Bone - Bone: dyn_coat_r_02"(ComponentPose=@n10, Alpha=1, Translation="0,0,0", Rotation="0,0,0", Scale="1,1,1") {
    rotationMode=BMM_Additive
    rotationSpace=BCS_BoneSpace
    translationMode=BMM_Additive
    translationSpace=BCS_BoneSpace
    alphaBoolBlend {blendOption=Linear}
    boneToModify {boneName=dyn_coat_r_02}
   }
   n12 = "Transform (Modify) Bone - Bone: dyn_coat_r_03"(ComponentPose=@n11, Alpha=1, Translation="0,0,0", Rotation="0,0,0", Scale="1,1,1") {
    rotationMode=BMM_Additive
    rotationSpace=BCS_BoneSpace
    translationMode=BMM_Additive
    translationSpace=BCS_BoneSpace
    alphaBoolBlend {blendOption=Linear}
    boneToModify {boneName=dyn_coat_r_03}
   }
   n13 = "Constraint - dyn_coat_bk_l_04 (RR)"!(ComponentPose=@n12, Alpha=1, ConstraintWeights_0=0.5, ConstraintWeights_1:float=0.5) -> (Pose) {
    alphaBoolBlend {blendOption=Linear}
    boneToModify {boneName=dyn_coat_bk_l_04}
    constraintSetup [{transformType=Rotation, targetBone {boneName=calf_l}}, {transformType=Rotation, targetBone {boneName=foot_l}}]
    constraintWeights [1,1]
   }
   n14 = "Constraint - dyn_coat_bk_r_04 (RR)"!(ComponentPose=@n13, Alpha=1, ConstraintWeights_0=0.5, ConstraintWeights_1:float=0.5) -> (Pose) {
    alphaBoolBlend {blendOption=Linear}
    boneToModify {boneName=dyn_coat_bk_r_04}
    constraintSetup [{transformType=Rotation, targetBone {boneName=calf_r}}, {transformType=Rotation, targetBone {boneName=foot_r}}]
    constraintWeights [1,1]
   }
   n15 = "Constraint - dyn_coat_bk_c_04 (RR)"!(ComponentPose=@n14, Alpha=1, ConstraintWeights_0=1, ConstraintWeights_1:float=1) -> (Pose) {
    alphaBoolBlend {blendOption=Linear}
    boneToModify {boneName=dyn_coat_bk_c_04}
    constraintSetup [{transformType=Rotation, targetBone {boneName=dyn_coat_bk_l_04}}, {transformType=Rotation, targetBone {boneName=dyn_coat_bk_r_04}}]
    constraintWeights [1,1]
   }
   n16 = "Constraint - dyn_coat_bk_l_05 (RR)"!(ComponentPose=@n15, Alpha=1, ConstraintWeights_0=0.5, ConstraintWeights_1:float=0.5) -> (Pose) {
    alphaBoolBlend {blendOption=Linear}
    boneToModify {boneName=dyn_coat_bk_l_05}
    constraintSetup [{transformType=Rotation, targetBone {boneName=calf_l}}, {transformType=Rotation, targetBone {boneName=foot_l}}]
    constraintWeights [1,1]
   }
   n17 = "Constraint - dyn_coat_bk_r_05 (RR)"!(ComponentPose=@n16, Alpha=1, ConstraintWeights_0=0.5, ConstraintWeights_1:float=0.5) -> (Pose) {
    alphaBoolBlend {blendOption=Linear}
    boneToModify {boneName=dyn_coat_bk_r_05}
    constraintSetup [{transformType=Rotation, targetBone {boneName=calf_r}}, {transformType=Rotation, targetBone {boneName=foot_r}}]
    constraintWeights [1,1]
   }
   "Constraint - dyn_coat_bk_c_05 (RR)"!(ComponentPose=@n17, Alpha=1, ConstraintWeights_0=1, ConstraintWeights_1:float=1) -> (Pose) {
    alphaBoolBlend {blendOption=Linear}
    boneToModify {boneName=dyn_coat_bk_c_05}
    constraintSetup [{transformType=Rotation, targetBone {boneName=dyn_coat_bk_l_04}}, {transformType=Rotation, targetBone {boneName=dyn_coat_bk_r_04}}]
    constraintWeights [1,1]
   }
   Local To Component()
   n18 = Get bIsSprinting()
   n19 = "Transform (Modify) Bone - Bone: cape_mid_02_r"!(ComponentPose, bAlphaBoolEnabled:bool=@n18, Translation="0,0,0", Rotation="2,0,0", Scale="1,1,1") -> (Pose) {
    alphaInputType=Bool
    rotationMode=BMM_Additive
    rotationSpace=BCS_BoneSpace
    scaleMode=BMM_Additive
    scaleSpace=BCS_BoneSpace
    alphaBoolBlend {blendInTime=0.2, blendOption=Linear, blendOutTime=1}
    boneToModify {boneName=cape_mid_02_r}
    rotation {pitch=2}
   }
   n20 = "Transform (Modify) Bone - Bone: cape_mid_04_r"!(ComponentPose=@n19, bAlphaBoolEnabled:bool=@n18, Translation="0,0,0", Rotation="4,0,0", Scale="1,1,1") -> (Pose) {
    alphaInputType=Bool
    rotationMode=BMM_Additive
    rotationSpace=BCS_BoneSpace
    scaleMode=BMM_Additive
    scaleSpace=BCS_BoneSpace
    alphaBoolBlend {blendInTime=0.2, blendOption=Linear, blendOutTime=1}
    boneToModify {boneName=cape_mid_04_r}
    rotation {pitch=4}
   }
   n21 = "Transform (Modify) Bone - Bone: cape_01_r"!(ComponentPose=@n20, bAlphaBoolEnabled:bool=@n18, Translation="0,0,0", Rotation="0,0,0", Scale="1,1,1") -> (Pose) {
    alphaInputType=Bool
    rotationMode=BMM_Additive
    rotationSpace=BCS_BoneSpace
    scaleMode=BMM_Additive
    scaleSpace=BCS_BoneSpace
    translationMode=BMM_Additive
    translationSpace=BCS_BoneSpace
    alphaBoolBlend {blendInTime=0.2, blendOption=Linear, blendOutTime=1}
    boneToModify {boneName=cape_01_r}
   }
   n22 = Get bIsSprinting()
   n23 = "Transform (Modify) Bone - Bone: cape_mid_02_r"(ComponentPose=@n21, bAlphaBoolEnabled=@n22, Translation="0,0,0", Rotation="2,0,0", Scale="1,1,1") {
    alphaInputType=Bool
    rotationMode=BMM_Additive
    rotationSpace=BCS_BoneSpace
    scaleMode=BMM_Additive
    scaleSpace=BCS_BoneSpace
    alphaBoolBlend {blendInTime=0.2, blendOption=Linear, blendOutTime=1}
    boneToModify {boneName=cape_mid_02_r}
    rotation {pitch=2}
   }
   n24 = "Transform (Modify) Bone - Bone: cape_mid_04_r"(ComponentPose=@n23, bAlphaBoolEnabled=@n22, Translation="0,0,0", Rotation="4,0,0", Scale="1,1,1") {
    alphaInputType=Bool
    rotationMode=BMM_Additive
    rotationSpace=BCS_BoneSpace
    scaleMode=BMM_Additive
    scaleSpace=BCS_BoneSpace
    alphaBoolBlend {blendInTime=0.2, blendOption=Linear, blendOutTime=1}
    boneToModify {boneName=cape_mid_04_r}
    rotation {pitch=4}
   }
   n25 = Get bIsSprinting()
   n26 = "Transform (Modify) Bone - Bone: cape_01_r"(ComponentPose=@n24, bAlphaBoolEnabled=@n25, Translation="0,0,0", Rotation="2,0,0", Scale="1,1,1") {
    alphaInputType=Bool
    rotationMode=BMM_Additive
    rotationSpace=BCS_BoneSpace
    scaleMode=BMM_Additive
    scaleSpace=BCS_BoneSpace
    alphaBoolBlend {blendInTime=0.6, blendOption=Linear, blendOutTime=0.1}
    boneToModify {boneName=cape_01_r}
    rotation {pitch=2}
   }
   "Transform (Modify) Bone - Bone: cape_04_r"!(ComponentPose=@n26, bAlphaBoolEnabled:bool=@n25, Translation="0,0,0", Rotation="3,7,0", Scale="0.7,1,1") -> (Pose) {
    alphaInputType=Bool
    rotationMode=BMM_Additive
    rotationSpace=BCS_BoneSpace
    scaleMode=BMM_Additive
    scaleSpace=BCS_BoneSpace
    alphaBoolBlend {blendInTime=0.2, blendOption=Linear, blendOutTime=1}
    boneToModify {boneName=cape_04_r}
    rotation {pitch=3, yaw=7}
    scale {x=0.7}
   }
   n27 = Get thigh L Rot() -> (thigh L Rot:double)
   n28 = "Transform (Modify) Bone - Bone: dyn_coat_front_l_01"(Alpha=@n27, Translation="0,0,0", Rotation="20,-20,0", Scale="1,1,1") {
    rotationMode=BMM_Additive
    rotationSpace=BCS_BoneSpace
    alphaBoolBlend {blendOption=Linear}
    alphaScaleBiasClamp {bInterpResult=true, inRange {max=-35, min=-15}}
    boneToModify {boneName=dyn_coat_front_l_01}
    rotation {pitch=20, yaw=-20}
   }
   n29 = "Transform (Modify) Bone - Bone: dyn_coat_front_l_02"(ComponentPose=@n28, Alpha=@n27, Translation="0,0,0", Rotation="20,-20,0", Scale="1,1,1") {
    rotationMode=BMM_Additive
    rotationSpace=BCS_BoneSpace
    alphaBoolBlend {blendOption=Linear}
    alphaScaleBiasClamp {bInterpResult=true, inRange {max=-35, min=-15}}
    boneToModify {boneName=dyn_coat_front_l_02}
    rotation {pitch=20, yaw=-20}
   }
   "Transform (Modify) Bone - Bone: dyn_coat_front_l_03"(ComponentPose=@n29, Alpha=@n27, Translation="0,0,0", Rotation="20,-20,0", Scale="1,1,1") {
    rotationMode=BMM_Additive
    rotationSpace=BCS_BoneSpace
    alphaBoolBlend {blendOption=Linear}
    alphaScaleBiasClamp {bInterpResult=true, inRange {max=-35, min=-15}}
    boneToModify {boneName=dyn_coat_front_l_03}
    rotation {pitch=20, yaw=-20}
   }
   n30 = "Transform (Modify) Bone - Bone: cape_01_r"(bAlphaBoolEnabled=True, Translation="0,0,0", Rotation="20.623,6.428,-0.001", Scale="1,1,1") {
    alphaInputType=Bool
    rotationMode=BMM_Additive
    rotationSpace=BCS_BoneSpace
    alphaBoolBlend {blendOption=Linear}
    boneToModify {boneName=cape_01_r}
    rotation=<rot 20.623 6.428 -0.001>
   }
   n31 = "Transform (Modify) Bone - Bone: cape_02_r"!(ComponentPose=@n30, bAlphaBoolEnabled:bool=True, Translation="0,0,0", Rotation="-0.750,6.803,-0.095", Scale="1,1,1") -> (Pose) {
    alphaInputType=Bool
    rotationMode=BMM_Additive
    rotationSpace=BCS_BoneSpace
    alphaBoolBlend {blendOption=Linear}
    boneToModify {boneName=cape_02_r}
    rotation=<rot -0.75 6.803 -0.095>
   }
   n32 = "Transform (Modify) Bone - Bone: cape_03_r"!(ComponentPose=@n31, bAlphaBoolEnabled:bool=True, Translation="0,0,0", Rotation="-0.002,12.516,-0.000", Scale="1,1,1") -> (Pose) {
    alphaInputType=Bool
    rotationMode=BMM_Additive
    rotationSpace=BCS_BoneSpace
    alphaBoolBlend {blendOption=Linear}
    boneToModify {boneName=cape_03_r}
    rotation=<rot -0.002 12.516 0>
   }
   "Transform (Modify) Bone - Bone: cape_04_r"(ComponentPose=@n32, bAlphaBoolEnabled=True, Translation="0,0,0", Rotation="-0.006,9.940,-0.012", Scale="1,1,1") {
    alphaInputType=Bool
    rotationMode=BMM_Additive
    rotationSpace=BCS_BoneSpace
    alphaBoolBlend {blendOption=Linear}
    boneToModify {boneName=cape_04_r}
    rotation=<rot -0.006 9.94 -0.012>
   }
   n33 = Get bIsSprinting()
   n34 = "Transform (Modify) Bone - Bone: cape_mid_02_l"!(ComponentPose, bAlphaBoolEnabled:bool=@n33, Translation="0,0,0", Rotation="2,0,0", Scale="1,1,1") -> (Pose) {
    alphaInputType=Bool
    rotationMode=BMM_Additive
    rotationSpace=BCS_BoneSpace
    scaleMode=BMM_Additive
    scaleSpace=BCS_BoneSpace
    alphaBoolBlend {blendInTime=0.2, blendOption=Linear, blendOutTime=1}
    boneToModify {boneName=cape_mid_02_l}
    rotation {pitch=2}
   }
   "Transform (Modify) Bone - Bone: cape_mid_04_l"!(ComponentPose=@n34, bAlphaBoolEnabled:bool=@n33, Translation="0,0,0", Rotation="4,0,0", Scale="1,1,1") -> (Pose) {
    alphaInputType=Bool
    rotationMode=BMM_Additive
    rotationSpace=BCS_BoneSpace
    scaleMode=BMM_Additive
    scaleSpace=BCS_BoneSpace
    alphaBoolBlend {blendInTime=0.2, blendOption=Linear, blendOutTime=1}
    boneToModify {boneName=cape_mid_04_l}
    rotation {pitch=4}
   }
   RigidBody!(ComponentPose, Alpha, OverrideWorldGravity:Vector=Get Gravity Override Cape() -> (Gravity Override Cape:Vector), ExternalForce) -> (Pose) {
    bClampLinearTranslationLimitToRefPose=true
    bOverrideWorldGravity=true
    overridePhysicsAsset=/BRCosmetics/Characters/Player/Male/Medium/Bodies/M_MED_Realm/Meshes/M_MED_Realm_SkirtStanding_Physics
    simulationSpace=BaseBoneSpace
    alphaBoolBlend {blendOption=Linear}
    baseBoneRef {boneName=root}
    componentAppliedLinearAccClamp=<100 100 100>
    componentLinearAccScale=<0.1 0.1 0.1>
    componentLinearVelScale=<0.1 0.1 0.1>
    overrideWorldGravity {z=-980}
   }
   n35 = "Transform (Modify) Bone - Bone: dyn_capeShoulderMSTR_l_01"!(
    ComponentPose=Local To Component(LocalPose=Use cached pose 'postSim'())
    bAlphaBoolEnabled:bool=false
    Translation="3,3,-3"
    Rotation="0,0,0"
    Scale="1,1,1"
   ) -> (Pose) {
    alphaInputType=Bool
    rotationMode=BMM_Additive
    rotationSpace=BCS_BoneSpace
    translationMode=BMM_Additive
    translationSpace=BCS_BoneSpace
    alphaBoolBlend {blendOption=Linear}
    boneToModify {boneName=dyn_capeShoulderMSTR_l_01}
    translation=<3 3 -3>
   }
   n36 = "Transform (Modify) Bone - Bone: dyn_capeShoulderMSTR_r_01"!(ComponentPose=@n35, bAlphaBoolEnabled:bool=false, Translation="-2,-2,-3", Rotation="-10,0,-10", Scale="1,1,1") -> (Pose) {
    alphaInputType=Bool
    rotationMode=BMM_Additive
    rotationSpace=BCS_BoneSpace
    translationMode=BMM_Additive
    translationSpace=BCS_BoneSpace
    alphaBoolBlend {blendOption=Linear}
    boneToModify {boneName=dyn_capeShoulderMSTR_r_01}
    rotation {pitch=-10, roll=-10}
    translation=<-2 -2 -3>
   }
   n37 = Get skydiveNoParachute() -> (skydiveNoParachute:bool)
   n38 = "Transform (Modify) Bone - Bone: cape_01_r"(ComponentPose=@n36, bAlphaBoolEnabled=@n37, Translation="0,0,20", Rotation="0,0,0", Scale="1,1,1") {
    alphaInputType=Bool
    rotationMode=BMM_Additive
    rotationSpace=BCS_BoneSpace
    translationMode=BMM_Additive
    translationSpace=BCS_BoneSpace
    alphaBoolBlend {blendOption=Linear}
    boneToModify {boneName=cape_01_r}
    translation {z=20}
   }
   n39 = "Transform (Modify) Bone - Bone: dyn_shoulderPad_l_01"!(ComponentPose=@n38, bAlphaBoolEnabled:bool=@n37, Translation="0,0,0", Rotation="0,0,0", Scale="1,1,1") -> (Pose) {
    alphaInputType=Bool
    rotationMode=BMM_Additive
    rotationSpace=BCS_BoneSpace
    translationMode=BMM_Additive
    translationSpace=BCS_BoneSpace
    alphaBoolBlend {blendOption=Linear}
    boneToModify {boneName=dyn_shoulderPad_l_01}
   }
   n40 = Get bIsParachuteOpened()
   n41 = "Transform (Modify) Bone - Bone: cape_01_r"(ComponentPose=@n39, bAlphaBoolEnabled=@n40, Translation="0,7,0", Rotation="0,0,-10", Scale="1,1,1") {
    alphaInputType=Bool
    rotationMode=BMM_Additive
    rotationSpace=BCS_BoneSpace
    translationMode=BMM_Additive
    translationSpace=BCS_BoneSpace
    alphaBoolBlend {blendOption=Linear}
    boneToModify {boneName=cape_01_r}
    rotation {roll=-10}
    translation {y=7}
   }
   n42 = "Transform (Modify) Bone - Bone: cape_01_l"!(ComponentPose=@n41, bAlphaBoolEnabled:bool=@n40, Translation="0,-8,0", Rotation="0,0,40", Scale="1,1,1") -> (Pose) {
    alphaInputType=Bool
    rotationMode=BMM_Additive
    rotationSpace=BCS_BoneSpace
    translationMode=BMM_Additive
    translationSpace=BCS_BoneSpace
    alphaBoolBlend {blendOption=Linear}
    boneToModify {boneName=cape_01_l}
    rotation {roll=40}
    translation {y=-8}
   }
   n43 = "Transform (Modify) Bone - Bone: cape_01_l"(ComponentPose=@n42, bAlphaBoolEnabled=@n33, Translation="0,0,10", Rotation="0,0,0", Scale="1,1,1") {
    alphaInputType=Bool
    rotationMode=BMM_Additive
    rotationSpace=BCS_BoneSpace
    scaleMode=BMM_Additive
    scaleSpace=BCS_BoneSpace
    translationMode=BMM_Additive
    translationSpace=BCS_BoneSpace
    alphaBoolBlend {blendInTime=0.2, blendOption=Linear, blendOutTime=1}
    boneToModify {boneName=cape_01_l}
    translation {z=10}
   }
   n44 = "Transform (Modify) Bone - Bone: cape_01_l"!(ComponentPose=@n43, Alpha:float=0, Translation="0,-5,4", Rotation="0,0,15", Scale="1,1,1") -> (Pose) {
    rotationMode=BMM_Additive
    rotationSpace=BCS_BoneSpace
    scaleMode=BMM_Additive
    scaleSpace=BCS_BoneSpace
    translationMode=BMM_Additive
    translationSpace=BCS_BoneSpace
    alphaBoolBlend {blendInTime=0.6, blendOption=Linear, blendOutTime=0.1}
    alphaScaleBiasClamp {bInterpResult=true, bMapRange=true}
    boneToModify {boneName=cape_01_l}
    rotation {roll=15}
    translation {y=-5, z=4}
   }
   n45 = "Transform (Modify) Bone - Bone: cape_03_l"(ComponentPose=@n44, Alpha=0, Translation="0,0,0", Rotation="0,0,0", Scale="1,1,1") {
    rotationMode=BMM_Additive
    rotationSpace=BCS_BoneSpace
    scaleMode=BMM_Additive
    scaleSpace=BCS_BoneSpace
    translationMode=BMM_Additive
    translationSpace=BCS_BoneSpace
    alphaBoolBlend {blendInTime=0.6, blendOption=Linear, blendOutTime=0.1}
    alphaScaleBiasClamp {bMapRange=true}
    boneToModify {boneName=cape_03_l}
   }
   n46 = "Transform (Modify) Bone - Bone: cape_04_l"(ComponentPose=@n45, Alpha=0, Translation="0,0,0", Rotation="-20,0,0", Scale="1,1,1") {
    rotationMode=BMM_Additive
    rotationSpace=BCS_BoneSpace
    scaleMode=BMM_Additive
    scaleSpace=BCS_BoneSpace
    translationMode=BMM_Additive
    translationSpace=BCS_BoneSpace
    alphaBoolBlend {blendInTime=0.6, blendOption=Linear, blendOutTime=0.1}
    alphaScaleBiasClamp {bMapRange=true}
    boneToModify {boneName=cape_04_l}
    rotation {pitch=-20}
   }
   n47 = "Constraint - dyn_lightSaber_01 (PPPP)"!(
    ComponentPose=@n46
    Alpha=1
    ConstraintWeights_0=0
    ConstraintWeights_1:float=0.2
    ConstraintWeights_2:float=0.4
    ConstraintWeights_3:float=0.4
   ) -> (Pose) {
    alphaBoolBlend {blendOption=Linear}
    boneToModify {boneName=dyn_lightSaber_01}
    constraintSetup [
     {targetBone {boneName=dyn_coat_front_l_01}}
     {targetBone {boneName=dyn_coat_l_01}}
     {targetBone {boneName=dyn_coat_front_l_03}}
     {targetBone {boneName=spine_01}}
    ]
    constraintWeights [1,1,1,1]
   }
   n48 = "Constraint - dyn_lightSaber_01 (PPP)"!(ComponentPose=@n47, Alpha=0, ConstraintWeights_0=0.5, ConstraintWeights_1:float=0, ConstraintWeights_2:float=0.5) -> (Pose) {
    alphaBoolBlend {blendOption=Linear}
    boneToModify {boneName=dyn_lightSaber_01}
    constraintSetup [{targetBone {boneName=dyn_coat_front_l_02}}, {targetBone {boneName=dyn_coat_l_02}}, {targetBone {boneName=dyn_coat_front_l_03}}]
    constraintWeights [1,1,1]
   }
   n49 = "Transform (Modify) Bone - Bone: dyn_lightSaber_01"(ComponentPose=@n48, Alpha=1, Translation="0,0,0", Rotation="0,0,0", Scale="1,1,1") {
    rotationMode=BMM_Additive
    rotationSpace=BCS_BoneSpace
    translationMode=BMM_Additive
    translationSpace=BCS_BoneSpace
    alphaBoolBlend {blendOption=Linear}
    boneToModify {boneName=dyn_lightSaber_01}
   }
   n50 = "Transform (Modify) Bone - Bone: dyn_lightSaber_02"(ComponentPose=@n49, Alpha=1, Translation="0,0,0", Rotation="6,0,0", Scale="1,1,1") {
    rotationMode=BMM_Additive
    rotationSpace=BCS_BoneSpace
    translationMode=BMM_Additive
    translationSpace=BCS_BoneSpace
    alphaBoolBlend {blendOption=Linear}
    boneToModify {boneName=dyn_lightSaber_02}
    rotation {pitch=6}
   }
   n51 = Anim Dynamics - Bone: dyn_lightSaber_01(ComponentPose=@n50, Alpha=0, ExternalForce="0,0,0") {
    angularDampingOverride=0.9
    angularSpringConstant=60
    bAngularSpring=true
    bOverrideAngularDamping=true
    bUseGravityOverride=true
    simulationSpace=RootRelative
    alphaBoolBlend {blendOption=Linear}
    boundBone {boneName=dyn_lightSaber_01}
    componentAppliedLinearAccClamp=<3000 3000 3000>
    componentLinearAccScale=<0.1 0.1 0.1>
    componentLinearVelScale=<0.1 0.1 0.1>
    gravityOverride {z=-980}
    physicsBodyDefinitions [{
      boundBone {boneName=dyn_lightSaber_01}
      boxExtents=<15 30 30>
      constraintSetup {angularLimitsMax=<0 0 5>, angularLimitsMin=<0 0 -5>, angularTarget=<0 0 0>}
      localJointOffset=<8 0 0>
     }]
    relativeSpaceBone {boneName=pelvis}
   }
   n52 = Anim Dynamics - Bone: dyn_lightSaber_02(ComponentPose=@n51, Alpha=1, ExternalForce="0,0,0") {
    angularDampingOverride=0.7
    angularSpringConstant=10
    bAngularSpring=true
    bOverrideAngularDamping=true
    bUseGravityOverride=true
    numSolverIterationsPreUpdate=3
    simulationSpace=RootRelative
    alphaBoolBlend {blendOption=Linear}
    boundBone {boneName=dyn_lightSaber_02}
    componentAppliedLinearAccClamp=<3000 3000 3000>
    componentLinearAccScale=<0.05 0.05 0.05>
    componentLinearVelScale=<0.05 0.05 0.05>
    gravityOverride {z=-980}
    physicsBodyDefinitions [{
      boundBone {boneName=dyn_lightSaber_02}
      boxExtents=<10 30 30>
      constraintSetup {angularLimitsMax=<0 0 8>, angularLimitsMin=<0 0 -8>, angularTarget=<0 0 0>}
      localJointOffset=<15 0 0>
     }]
    relativeSpaceBone {boneName=pelvis}
   }
   n53 = Get isCrouchStopped() -> (isCrouchStopped:bool)
   n54 = "Transform (Modify) Bone - Bone: dyn_lightSaber_01"!(ComponentPose=@n52, bAlphaBoolEnabled:bool=@n53, Translation="0,0,0", Rotation="20,0,0", Scale="1,1,1") -> (Pose) {
    alphaInputType=Bool
    rotationMode=BMM_Additive
    rotationSpace=BCS_BoneSpace
    translationMode=BMM_Additive
    translationSpace=BCS_BoneSpace
    alphaBoolBlend {blendOption=Linear, blendOutTime=1}
    boneToModify {boneName=dyn_lightSaber_01}
    rotation {pitch=20}
   }
   n55 = "Transform (Modify) Bone - Bone: dyn_lightSaber_02"!(ComponentPose=@n54, bAlphaBoolEnabled:bool=@n53, Translation="0,0,0", Rotation="7,0,0", Scale="1,1,1") -> (Pose) {
    alphaInputType=Bool
    rotationMode=BMM_Additive
    rotationSpace=BCS_BoneSpace
    translationMode=BMM_Additive
    translationSpace=BCS_BoneSpace
    alphaBoolBlend {blendOption=Linear, blendOutTime=1}
    boneToModify {boneName=dyn_lightSaber_02}
    rotation {pitch=7}
   }
   Output Pose(Result=Component To Local(ComponentPose=@n55)) {layerGroup=None}
   n56 = Control Rig(Source=Copy Pose From Mesh() {bUseAttachedParent=true}) {
    alphaBoolBlend {blendOption=Linear}
    controlRigAssetReference {blueprintRigClass=/BRCosmetics/Characters/Player/Male/Medium/Bodies/M_MED_Realm/Meshes/M_MED_Realm_Deform_CtrlRig_twists.M_MED_Realm_Deform_CtrlRig_twists_C}
   }
   n57 = Control Rig!(
    Source=@n56
    Alpha
    KneeRot_fr_l_114deg:Vector
    KneeTrans_fr_l_114deg:Vector="0.258,-3.591,0.009"
    KneeScale_fr_l_114deg:Vector="1,1,1"
    KneeScale_fr_l_130deg:Vector="1,1,1"
    KneeRot_fr_l_130deg:Vector="0,0,-51.2"
    KneeTrans_fr_l_130deg:Vector="0.258,-3.591,0.009"
   ) -> (Pose) {
    alphaBoolBlend {blendOption=Linear}
    controlRigAssetReference {blueprintRigClass=/BRCosmetics/Characters/Player/Male/Medium/Bodies/M_MED_Realm/Meshes/M_MED_Realm_Deform_CtrlRig_sculpters.M_MED_Realm_Deform_CtrlRig_sculpters_C}
   }
   Save cached pose 'controlRigCache'(Pose=@n57) {cacheName=controlRigCache}
   n58 = RigidBody(ComponentPose=Local To Component(LocalPose=Use cached pose 'standingPoseCache'()), OverrideWorldGravity=Get Gravity Override Cape()) {
    bClampLinearTranslationLimitToRefPose=true
    bOverrideWorldGravity=true
    overridePhysicsAsset=/BRCosmetics/Characters/Player/Male/Medium/Bodies/M_MED_Realm/Meshes/M_MED_Realm_Cape_Physics
    simulationSpace=BaseBoneSpace
    alphaBoolBlend {blendOption=Linear}
    baseBoneRef {boneName=root}
    componentAppliedLinearAccClamp=<3000 3000 3000>
    componentLinearAccScale=<0.05 0.05 0.3>
    componentLinearVelScale=<0.05 0.05 0.3>
    overrideWorldGravity {z=-980}
    simSpaceSettings {
     maxAngularAcceleration=50
     maxAngularVelocity=1
     maxLinearAcceleration=300
     maxLinearVelocity=150
     velocityScaleZ=0
     externalLinearDragV=<0.2 0.2>
    }
   }
   n59 = RigidBody(ComponentPose=Local To Component(LocalPose=Use cached pose 'standingPoseCache'()), OverrideWorldGravity=Get Gravity Override Cape()) {
    bClampLinearTranslationLimitToRefPose=true
    bOverrideWorldGravity=true
    overridePhysicsAsset=/BRCosmetics/Characters/Player/Male/Medium/Bodies/M_MED_Realm/Meshes/M_MED_RealmBackBling_Cape_Physics
    simulationSpace=BaseBoneSpace
    alphaBoolBlend {blendOption=Linear}
    baseBoneRef {boneName=root}
    componentAppliedLinearAccClamp=<3000 3000 3000>
    componentLinearAccScale=<0.05 0.05 0.3>
    componentLinearVelScale=<0.05 0.05 0.3>
    overrideWorldGravity {z=-980}
    simSpaceSettings {
     maxAngularAcceleration=50
     maxAngularVelocity=1
     maxLinearAcceleration=300
     maxLinearVelocity=150
     velocityScaleZ=0
     externalLinearDragV=<0.2 0.2>
    }
   }
   n60 = Blend Poses by bool(
    BlendPose_0=Use cached pose 'crouchPose'()
    BlendPose_1=Blend Poses by bool(
     BlendPose_0=Component To Local(ComponentPose=@n59)
     BlendPose_1=Component To Local(ComponentPose=@n58)
     bActiveValue=Get Is Wearing Backbling()
    ) {blendType=Linear}
    BlendTime_1=0.3
    bActiveValue=Get isCrouchStopped()
   ) {blendType=Linear}
   Save cached pose 'postSim'(Pose=Component To Local(ComponentPose=Local To Component(LocalPose=@n60))) {cacheName=postSim}
   n61 = Use cached pose 'controlRigCache'()
   Save cached pose 'standingPoseCache'(Pose=Component To Local(ComponentPose=Local To Component(LocalPose=@n61))) {cacheName=standingPoseCache}
   n62 = "Transform (Modify) Bone - Bone: dyn_coat_front_l_01"(
    ComponentPose=Local To Component(LocalPose=Use cached pose 'controlRigCache'())
    Alpha=1
    Translation="-4,-3,-2"
    Rotation="20,-78.117,-85"
    Scale="1,1,1"
   ) {
    rotationMode=BMM_Additive
    rotationSpace=BCS_BoneSpace
    scaleMode=BMM_Additive
    scaleSpace=BCS_BoneSpace
    translationMode=BMM_Additive
    translationSpace=BCS_BoneSpace
    alphaBoolBlend {blendOption=Linear}
    boneToModify {boneName=dyn_coat_front_l_01}
    rotation=<rot 20 -78.117 -85>
    translation=<-4 -3 -2>
   }
   n63 = "Transform (Modify) Bone - Bone: dyn_coat_front_l_02"(ComponentPose=@n62, Alpha=1, Translation="0,0,0", Rotation="-60.832,-4.831,0.145", Scale="1,1,1") {
    rotationMode=BMM_Additive
    rotationSpace=BCS_BoneSpace
    scaleMode=BMM_Additive
    scaleSpace=BCS_BoneSpace
    translationMode=BMM_Additive
    translationSpace=BCS_BoneSpace
    alphaBoolBlend {blendOption=Linear}
    boneToModify {boneName=dyn_coat_front_l_02}
    rotation=<rot -60.832 -4.831 0.145>
   }
   n64 = "Transform (Modify) Bone - Bone: dyn_coat_front_l_03"(ComponentPose=@n63, Alpha=1, Translation="0,0,0", Rotation="0.761,-3.678,-0.538", Scale="1,1,1") {
    rotationMode=BMM_Additive
    rotationSpace=BCS_BoneSpace
    scaleMode=BMM_Additive
    scaleSpace=BCS_BoneSpace
    translationMode=BMM_Additive
    translationSpace=BCS_BoneSpace
    alphaBoolBlend {blendOption=Linear}
    boneToModify {boneName=dyn_coat_front_l_03}
    rotation=<rot 0.761 -3.678 -0.538>
   }
   n65 = "Transform (Modify) Bone - Bone: dyn_coat_front_l_05"(ComponentPose=@n64, Alpha=1, Translation="0,0,0", Rotation="46.711,-12.031,-8.207", Scale="1,1,1") {
    rotationMode=BMM_Additive
    rotationSpace=BCS_BoneSpace
    scaleMode=BMM_Additive
    scaleSpace=BCS_BoneSpace
    translationMode=BMM_Additive
    translationSpace=BCS_BoneSpace
    alphaBoolBlend {blendOption=Linear}
    boneToModify {boneName=dyn_coat_front_l_05}
    rotation=<rot 46.711 -12.031 -8.207>
   }
   n66 = "Transform (Modify) Bone - Bone: dyn_coat_front_l_06"(ComponentPose=@n65, Alpha=1, Translation="0,0,0", Rotation="24.981,-8.676,-6.168", Scale="1,1,1") {
    rotationMode=BMM_Additive
    rotationSpace=BCS_BoneSpace
    translationMode=BMM_Additive
    translationSpace=BCS_BoneSpace
    alphaBoolBlend {blendOption=Linear}
    boneToModify {boneName=dyn_coat_front_l_06}
    rotation=<rot 24.981 -8.676 -6.168>
   }
   n67 = "Transform (Modify) Bone - Bone: dyn_coat_l_01"(ComponentPose=@n66, Alpha=1, Translation="-4,0,0", Rotation="8,-12,-60", Scale="1,1,1") {
    rotationMode=BMM_Additive
    rotationSpace=BCS_BoneSpace
    scaleMode=BMM_Additive
    scaleSpace=BCS_BoneSpace
    translationMode=BMM_Additive
    translationSpace=BCS_BoneSpace
    alphaBoolBlend {blendOption=Linear}
    boneToModify {boneName=dyn_coat_l_01}
    rotation=<rot 8 -12 -60>
    translation {x=-4}
   }
   n68 = "Transform (Modify) Bone - Bone: dyn_coat_l_02"(ComponentPose=@n67, Alpha=1, Translation="0,0,0", Rotation="-11.103,-6.600,1.274", Scale="1,1,1") {
    rotationMode=BMM_Additive
    rotationSpace=BCS_BoneSpace
    scaleMode=BMM_Additive
    scaleSpace=BCS_BoneSpace
    translationMode=BMM_Additive
    translationSpace=BCS_BoneSpace
    alphaBoolBlend {blendOption=Linear}
    boneToModify {boneName=dyn_coat_l_02}
    rotation=<rot -11.103 -6.6 1.274>
   }
   n69 = "Transform (Modify) Bone - Bone: dyn_coat_l_04"(ComponentPose=@n68, Alpha=1, Translation="0,0,0", Rotation="18.060,0.026,-0.009", Scale="1,1,1") {
    rotationMode=BMM_Additive
    rotationSpace=BCS_BoneSpace
    scaleMode=BMM_Additive
    scaleSpace=BCS_BoneSpace
    translationMode=BMM_Additive
    translationSpace=BCS_BoneSpace
    alphaBoolBlend {blendOption=Linear}
    boneToModify {boneName=dyn_coat_l_04}
    rotation=<rot 18.06 0.026 -0.009>
   }
   n70 = "Transform (Modify) Bone - Bone: dyn_coat_l_05"(ComponentPose=@n69, Alpha=1, Translation="0,0,0", Rotation="67.291,0.075,2.919", Scale="1,1,1") {
    rotationMode=BMM_Additive
    rotationSpace=BCS_BoneSpace
    scaleMode=BMM_Additive
    scaleSpace=BCS_BoneSpace
    translationMode=BMM_Additive
    translationSpace=BCS_BoneSpace
    alphaBoolBlend {blendOption=Linear}
    boneToModify {boneName=dyn_coat_l_05}
    rotation=<rot 67.291 0.075 2.919>
   }
   n71 = "Transform (Modify) Bone - Bone: dyn_coat_bk_l_02"(ComponentPose=@n70, Alpha=1, Translation="0,0,0", Rotation="0.001,-7.847,0.001", Scale="1,1,1") {
    rotationMode=BMM_Additive
    rotationSpace=BCS_BoneSpace
    scaleMode=BMM_Additive
    scaleSpace=BCS_BoneSpace
    translationMode=BMM_Additive
    translationSpace=BCS_BoneSpace
    alphaBoolBlend {blendOption=Linear}
    boneToModify {boneName=dyn_coat_bk_l_02}
    rotation=<rot 0.001 -7.847 0.001>
   }
   n72 = "Transform (Modify) Bone - Bone: dyn_coat_bk_l_04"(ComponentPose=@n71, Alpha=1, Translation="0,0,0", Rotation="49.784,-0.006,-0.042", Scale="1,1,1") {
    rotationMode=BMM_Additive
    rotationSpace=BCS_BoneSpace
    scaleMode=BMM_Additive
    scaleSpace=BCS_BoneSpace
    translationMode=BMM_Additive
    translationSpace=BCS_BoneSpace
    alphaBoolBlend {blendOption=Linear}
    boneToModify {boneName=dyn_coat_bk_l_04}
    rotation=<rot 49.784 -0.006 -0.042>
   }
   n73 = "Transform (Modify) Bone - Bone: dyn_coat_bk_l_05"(ComponentPose=@n72, Alpha=1, Translation="0,0,0", Rotation="40.896,-0.398,0.903", Scale="1,1,1") {
    rotationMode=BMM_Additive
    rotationSpace=BCS_BoneSpace
    scaleMode=BMM_Additive
    scaleSpace=BCS_BoneSpace
    translationMode=BMM_Additive
    translationSpace=BCS_BoneSpace
    alphaBoolBlend {blendOption=Linear}
    boneToModify {boneName=dyn_coat_bk_l_05}
    rotation=<rot 40.896 -0.398 0.903>
   }
   n74 = "Transform (Modify) Bone - Bone: dyn_coat_bk_l_06"(ComponentPose=@n73, Alpha=1, Translation="0,0,0", Rotation="0,0,0", Scale="1,1,1") {
    rotationMode=BMM_Additive
    rotationSpace=BCS_BoneSpace
    scaleMode=BMM_Additive
    scaleSpace=BCS_BoneSpace
    translationMode=BMM_Additive
    translationSpace=BCS_BoneSpace
    alphaBoolBlend {blendOption=Linear}
    boneToModify {boneName=dyn_coat_bk_l_06}
   }
   n75 = "Transform (Modify) Bone - Bone: dyn_coat_bk_c_03"(ComponentPose=@n74, Alpha=1, Translation="0,0,0", Rotation="14.531,0.009,0.011", Scale="1,1,1") {
    rotationMode=BMM_Additive
    rotationSpace=BCS_BoneSpace
    scaleMode=BMM_Additive
    scaleSpace=BCS_BoneSpace
    translationMode=BMM_Additive
    translationSpace=BCS_BoneSpace
    alphaBoolBlend {blendOption=Linear}
    boneToModify {boneName=dyn_coat_bk_c_03}
    rotation=<rot 14.531 0.009 0.011>
   }
   n76 = "Transform (Modify) Bone - Bone: dyn_coat_bk_c_04"(ComponentPose=@n75, Alpha=1, Translation="0,0,0", Rotation="46.462,-0.012,-0.002", Scale="1,1,1") {
    rotationMode=BMM_Additive
    rotationSpace=BCS_BoneSpace
    scaleMode=BMM_Additive
    scaleSpace=BCS_BoneSpace
    translationMode=BMM_Additive
    translationSpace=BCS_BoneSpace
    alphaBoolBlend {blendOption=Linear}
    boneToModify {boneName=dyn_coat_bk_c_04}
    rotation=<rot 46.462 -0.012 -0.002>
   }
   n77 = "Transform (Modify) Bone - Bone: dyn_coat_bk_c_05"(ComponentPose=@n76, Alpha=1, Translation="0,0,0", Rotation="35.023,0.372,2.256", Scale="1,1,1") {
    rotationMode=BMM_Additive
    rotationSpace=BCS_BoneSpace
    scaleMode=BMM_Additive
    scaleSpace=BCS_BoneSpace
    translationMode=BMM_Additive
    translationSpace=BCS_BoneSpace
    alphaBoolBlend {blendOption=Linear}
    boneToModify {boneName=dyn_coat_bk_c_05}
    rotation=<rot 35.023 0.372 2.256>
   }
   n78 = "Transform (Modify) Bone - Bone: dyn_coat_bk_c_06"(ComponentPose=@n77, Alpha=1, Translation="0,0,0", Rotation="0,0,0", Scale="1,1,1") {
    rotationMode=BMM_Additive
    rotationSpace=BCS_BoneSpace
    scaleMode=BMM_Additive
    scaleSpace=BCS_BoneSpace
    translationMode=BMM_Additive
    translationSpace=BCS_BoneSpace
    alphaBoolBlend {blendOption=Linear}
    boneToModify {boneName=dyn_coat_bk_c_06}
   }
   n79 = "Transform (Modify) Bone - Bone: dyn_coat_front_r_01"(ComponentPose=@n78, Alpha=1, Translation="0,0,0", Rotation="12.650,-29.725,-18.944", Scale="1,1,1") {
    rotationMode=BMM_Additive
    rotationSpace=BCS_BoneSpace
    scaleMode=BMM_Additive
    scaleSpace=BCS_BoneSpace
    translationMode=BMM_Additive
    translationSpace=BCS_BoneSpace
    alphaBoolBlend {blendOption=Linear}
    boneToModify {boneName=dyn_coat_front_r_01}
    rotation=<rot 12.65 -29.725 -18.944>
   }
   n80 = "Transform (Modify) Bone - Bone: dyn_coat_front_r_02"(ComponentPose=@n79, Alpha=1, Translation="0,0,0", Rotation="0.740,-4.473,0", Scale="1,1,1") {
    rotationMode=BMM_Additive
    rotationSpace=BCS_BoneSpace
    scaleMode=BMM_Additive
    scaleSpace=BCS_BoneSpace
    translationMode=BMM_Additive
    translationSpace=BCS_BoneSpace
    alphaBoolBlend {blendOption=Linear}
    boneToModify {boneName=dyn_coat_front_r_02}
    rotation {pitch=0.74, yaw=-4.473}
   }
   n81 = "Transform (Modify) Bone - Bone: dyn_coat_front_r_03"(ComponentPose=@n80, Alpha=1, Translation="0,0,0", Rotation="-8.654,0.015,0.009", Scale="1,1,1") {
    rotationMode=BMM_Additive
    rotationSpace=BCS_BoneSpace
    scaleMode=BMM_Additive
    scaleSpace=BCS_BoneSpace
    translationMode=BMM_Additive
    translationSpace=BCS_BoneSpace
    alphaBoolBlend {blendOption=Linear}
    boneToModify {boneName=dyn_coat_front_r_03}
    rotation=<rot -8.654 0.015 0.009>
   }
   n82 = "Transform (Modify) Bone - Bone: dyn_coat_front_r_04"(ComponentPose=@n81, Alpha=1, Translation="0,0,0", Rotation="17.183,-13.904,-1.581", Scale="1,1,1") {
    rotationMode=BMM_Additive
    rotationSpace=BCS_BoneSpace
    scaleMode=BMM_Additive
    scaleSpace=BCS_BoneSpace
    translationMode=BMM_Additive
    translationSpace=BCS_BoneSpace
    alphaBoolBlend {blendOption=Linear}
    boneToModify {boneName=dyn_coat_front_r_04}
    rotation=<rot 17.183 -13.904 -1.581>
   }
   n83 = "Transform (Modify) Bone - Bone: dyn_coat_front_r_05"(ComponentPose=@n82, Alpha=1, Translation="0,0,0", Rotation="15.453,-0.561,-6.423", Scale="1,1,1") {
    rotationMode=BMM_Additive
    rotationSpace=BCS_BoneSpace
    scaleMode=BMM_Additive
    scaleSpace=BCS_BoneSpace
    translationMode=BMM_Additive
    translationSpace=BCS_BoneSpace
    alphaBoolBlend {blendOption=Linear}
    boneToModify {boneName=dyn_coat_front_r_05}
    rotation=<rot 15.453 -0.561 -6.423>
   }
   n84 = "Transform (Modify) Bone - Bone: dyn_coat_r_01"(ComponentPose=@n83, Alpha=1, Translation="0,0,0", Rotation="7.235,-4.132,-0.000", Scale="1,1,1") {
    rotationMode=BMM_Additive
    rotationSpace=BCS_BoneSpace
    scaleMode=BMM_Additive
    scaleSpace=BCS_BoneSpace
    translationMode=BMM_Additive
    translationSpace=BCS_BoneSpace
    alphaBoolBlend {blendOption=Linear}
    boneToModify {boneName=dyn_coat_r_01}
    rotation=<rot 7.235 -4.132 0>
   }
   n85 = "Transform (Modify) Bone - Bone: dyn_coat_r_02"(ComponentPose=@n84, Alpha=1, Translation="0,0,0", Rotation="5.238,-4.652,-0.232", Scale="1,1,1") {
    rotationMode=BMM_Additive
    rotationSpace=BCS_BoneSpace
    scaleMode=BMM_Additive
    scaleSpace=BCS_BoneSpace
    translationMode=BMM_Additive
    translationSpace=BCS_BoneSpace
    alphaBoolBlend {blendOption=Linear}
    boneToModify {boneName=dyn_coat_r_02}
    rotation=<rot 5.238 -4.652 -0.232>
   }
   n86 = "Transform (Modify) Bone - Bone: dyn_coat_r_03"(ComponentPose=@n85, Alpha=1, Translation="0,0,0", Rotation="7.773,-9.833,-2.686", Scale="1,1,1") {
    rotationMode=BMM_Additive
    rotationSpace=BCS_BoneSpace
    scaleMode=BMM_Additive
    scaleSpace=BCS_BoneSpace
    translationMode=BMM_Additive
    translationSpace=BCS_BoneSpace
    alphaBoolBlend {blendOption=Linear}
    boneToModify {boneName=dyn_coat_r_03}
    rotation=<rot 7.773 -9.833 -2.686>
   }
   n87 = "Transform (Modify) Bone - Bone: dyn_coat_r_04"(ComponentPose=@n86, Alpha=1, Translation="0,0,0", Rotation="16.472,-10.258,-2.764", Scale="1,1,1") {
    rotationMode=BMM_Additive
    rotationSpace=BCS_BoneSpace
    scaleMode=BMM_Additive
    scaleSpace=BCS_BoneSpace
    translationMode=BMM_Additive
    translationSpace=BCS_BoneSpace
    alphaBoolBlend {blendOption=Linear}
    boneToModify {boneName=dyn_coat_r_04}
    rotation=<rot 16.472 -10.258 -2.764>
   }
   n88 = "Transform (Modify) Bone - Bone: dyn_coat_r_05"(ComponentPose=@n87, Alpha=1, Translation="0,0,0", Rotation="16.472,-10.258,-2.764", Scale="1,1,1") {
    rotationMode=BMM_Additive
    rotationSpace=BCS_BoneSpace
    scaleMode=BMM_Additive
    scaleSpace=BCS_BoneSpace
    translationMode=BMM_Additive
    translationSpace=BCS_BoneSpace
    alphaBoolBlend {blendOption=Linear}
    boneToModify {boneName=dyn_coat_r_05}
    rotation=<rot 16.472 -10.258 -2.764>
   }
   n89 = "Transform (Modify) Bone - Bone: dyn_coat_bk_r_03"(ComponentPose=@n88, Alpha=1, Translation="0,0,0", Rotation="66.850,-4.107,-0.014", Scale="1,1,1") {
    rotationMode=BMM_Additive
    rotationSpace=BCS_BoneSpace
    scaleMode=BMM_Additive
    scaleSpace=BCS_BoneSpace
    translationMode=BMM_Additive
    translationSpace=BCS_BoneSpace
    alphaBoolBlend {blendOption=Linear}
    boneToModify {boneName=dyn_coat_bk_r_03}
    rotation=<rot 66.85 -4.107 -0.014>
   }
   n90 = "Transform (Modify) Bone - Bone: dyn_coat_bk_r_04"(ComponentPose=@n89, Alpha=1, Translation="0,0,0", Rotation="7.235,-4.132,-0.000", Scale="1,1,1") {
    rotationMode=BMM_Additive
    rotationSpace=BCS_BoneSpace
    scaleMode=BMM_Additive
    scaleSpace=BCS_BoneSpace
    translationMode=BMM_Additive
    translationSpace=BCS_BoneSpace
    alphaBoolBlend {blendOption=Linear}
    boneToModify {boneName=dyn_coat_bk_r_04}
    rotation=<rot 7.235 -4.132 0>
   }
   n91 = "Transform (Modify) Bone - Bone: dyn_coat_bk_r_05"(ComponentPose=@n90, Alpha=1, Translation="0,0,0", Rotation="0,0,0", Scale="1,1,1") {
    rotationMode=BMM_Additive
    rotationSpace=BCS_BoneSpace
    scaleMode=BMM_Additive
    scaleSpace=BCS_BoneSpace
    translationMode=BMM_Additive
    translationSpace=BCS_BoneSpace
    alphaBoolBlend {blendOption=Linear}
    boneToModify {boneName=dyn_coat_bk_r_05}
   }
   n92 = "Transform (Modify) Bone - Bone: dyn_coat_bk_r_06"(ComponentPose=@n91, Alpha=1, Translation="0,0,0", Rotation="0,0,0", Scale="1,1,1") {
    rotationMode=BMM_Additive
    rotationSpace=BCS_BoneSpace
    scaleMode=BMM_Additive
    scaleSpace=BCS_BoneSpace
    translationMode=BMM_Additive
    translationSpace=BCS_BoneSpace
    alphaBoolBlend {blendOption=Linear}
    boneToModify {boneName=dyn_coat_bk_r_06}
   }
   n93 = "Transform (Modify) Bone - Bone: dyn_lightSaber_01"!(ComponentPose=@n92, Alpha:float=1, Translation="0,0,0", Rotation="25,-45,-40", Scale="1,1,1") -> (Pose) {
    rotationMode=BMM_Additive
    rotationSpace=BCS_BoneSpace
    scaleMode=BMM_Additive
    scaleSpace=BCS_BoneSpace
    translationMode=BMM_Additive
    translationSpace=BCS_BoneSpace
    alphaBoolBlend {blendOption=Linear}
    boneToModify {boneName=dyn_lightSaber_01}
    rotation=<rot 25 -45 -40>
   }
   n94 = "Transform (Modify) Bone - Bone: dyn_lightSaber_02"!(ComponentPose=@n93, Alpha:float=1, Translation="0,0,0", Rotation="-27,0,0", Scale="1,1,1") -> (Pose) {
    rotationMode=BMM_Additive
    rotationSpace=BCS_BoneSpace
    scaleMode=BMM_Additive
    scaleSpace=BCS_BoneSpace
    translationMode=BMM_Additive
    translationSpace=BCS_BoneSpace
    alphaBoolBlend {blendOption=Linear}
    boneToModify {boneName=dyn_lightSaber_02}
    rotation {pitch=-27}
   }
   n95 = RigidBody(ComponentPose=@n94, Alpha=0, OverrideWorldGravity=Get Gravity Override Cape()) {
    bClampLinearTranslationLimitToRefPose=true
    bOverrideWorldGravity=true
    overridePhysicsAsset=/BRCosmetics/Characters/Player/Male/Medium/Bodies/M_MED_Realm/Meshes/M_MED_Realm_SkirtCrouching_Physics
    simulationSpace=BaseBoneSpace
    alphaBoolBlend {blendOption=Linear}
    baseBoneRef {boneName=root}
    componentAppliedLinearAccClamp=<100 100 100>
    componentLinearAccScale=<0.1 0.1 0.1>
    componentLinearVelScale=<0.1 0.1 0.1>
    overrideWorldGravity {z=-980}
   }
   n96 = Anim Dynamics - Bone: dyn_coat_bk_r_06(ComponentPose=@n95, Alpha=1, ExternalForce="0,0,0") {
    angularDampingOverride=1
    angularSpringConstant=20000
    bAngularSpring=true
    bOverrideAngularDamping=true
    bOverrideLinearDamping=true
    bUseGravityOverride=true
    gravityScale=0.35
    linearDampingOverride=0.7
    simulationSpace=BoneRelative
    alphaBoolBlend {blendOption=Linear}
    boundBone {boneName=dyn_coat_bk_r_06}
    componentLinearAccScale=<1 1 1>
    componentLinearVelScale=<1 1 1>
    gravityOverride {z=-980}
    physicsBodyDefinitions [{
      boundBone {boneName=dyn_coat_bk_r_06}
      boxExtents=<15 30 30>
      constraintSetup {angularLimitsMax=<0 0 0>, angularLimitsMin=<0 0 0>, angularTarget=<0 0 0>}
      localJointOffset=<-10 0 0>
     }]
    relativeSpaceBone {boneName=root}
   }
   n97 = Anim Dynamics - Bone: dyn_coat_r_06(ComponentPose=@n96, Alpha=1, ExternalForce="0,0,0") {
    angularDampingOverride=1
    angularSpringConstant=20000
    bAngularSpring=true
    bOverrideAngularDamping=true
    bOverrideLinearDamping=true
    bUseGravityOverride=true
    gravityScale=0.35
    linearDampingOverride=0.7
    simulationSpace=BoneRelative
    alphaBoolBlend {blendOption=Linear}
    boundBone {boneName=dyn_coat_r_06}
    componentLinearAccScale=<1 1 1>
    componentLinearVelScale=<1 1 1>
    gravityOverride {z=-980}
    physicsBodyDefinitions [{
      boundBone {boneName=dyn_coat_r_06}
      boxExtents=<15 30 30>
      constraintSetup {angularLimitsMax=<0 0 0>, angularLimitsMin=<0 0 0>, angularTarget=<0 0 0>}
      localJointOffset=<-10 0 0>
     }]
    relativeSpaceBone {boneName=root}
   }
   n98 = Anim Dynamics - Bone: dyn_coat_front_r_06(ComponentPose=@n97, Alpha=1, ExternalForce="0,0,0") {
    angularDampingOverride=1
    angularSpringConstant=20000
    bAngularSpring=true
    bOverrideAngularDamping=true
    bOverrideLinearDamping=true
    bUseGravityOverride=true
    gravityScale=0.35
    linearDampingOverride=0.7
    simulationSpace=BoneRelative
    alphaBoolBlend {blendOption=Linear}
    boundBone {boneName=dyn_coat_front_r_06}
    componentLinearAccScale=<1 1 1>
    componentLinearVelScale=<1 1 1>
    gravityOverride {z=-980}
    physicsBodyDefinitions [{
      boundBone {boneName=dyn_coat_front_r_06}
      boxExtents=<15 30 30>
      constraintSetup {angularLimitsMax=<0 0 0>, angularLimitsMin=<0 0 0>, angularTarget=<0 0 0>}
      localJointOffset=<-10 0 0>
     }]
    relativeSpaceBone {boneName=root}
   }
   n99 = Anim Dynamics - Bone: dyn_coat_bk_c_06(ComponentPose=@n98, Alpha=1, ExternalForce="0,0,0") {
    angularDampingOverride=1
    angularSpringConstant=20000
    bAngularSpring=true
    bOverrideAngularDamping=true
    bOverrideLinearDamping=true
    bUseGravityOverride=true
    gravityScale=0.35
    linearDampingOverride=0.7
    simulationSpace=BoneRelative
    alphaBoolBlend {blendOption=Linear}
    boundBone {boneName=dyn_coat_bk_c_06}
    componentLinearAccScale=<1 1 1>
    componentLinearVelScale=<1 1 1>
    gravityOverride {z=-980}
    physicsBodyDefinitions [{
      boundBone {boneName=dyn_coat_bk_c_06}
      boxExtents=<15 30 30>
      constraintSetup {angularLimitsMax=<0 0 0>, angularLimitsMin=<0 0 0>, angularTarget=<0 0 0>}
      localJointOffset=<10 0 0>
     }]
    relativeSpaceBone {boneName=root}
   }
   n100 = Anim Dynamics - Bone: dyn_coat_bk_l_06(ComponentPose=@n99, Alpha=1, ExternalForce="0,0,0") {
    angularDampingOverride=1
    angularSpringConstant=20000
    bAngularSpring=true
    bOverrideAngularDamping=true
    bOverrideLinearDamping=true
    bUseGravityOverride=true
    gravityScale=0.35
    linearDampingOverride=0.7
    simulationSpace=BoneRelative
    alphaBoolBlend {blendOption=Linear}
    boundBone {boneName=dyn_coat_bk_l_06}
    componentLinearAccScale=<1 1 1>
    componentLinearVelScale=<1 1 1>
    gravityOverride {z=-980}
    physicsBodyDefinitions [{
      boundBone {boneName=dyn_coat_bk_l_06}
      boxExtents=<15 30 30>
      constraintSetup {angularLimitsMax=<0 0 0>, angularLimitsMin=<0 0 0>, angularTarget=<0 0 0>}
      localJointOffset=<10 0 0>
     }]
    relativeSpaceBone {boneName=root}
   }
   n101 = Anim Dynamics - Bone: dyn_coat_l_06(ComponentPose=@n100, Alpha=1, ExternalForce="0,0,0") {
    angularDampingOverride=1
    angularSpringConstant=20000
    bAngularSpring=true
    bOverrideAngularDamping=true
    bOverrideLinearDamping=true
    bUseGravityOverride=true
    gravityScale=0.35
    linearDampingOverride=0.7
    simulationSpace=BoneRelative
    alphaBoolBlend {blendOption=Linear}
    boundBone {boneName=dyn_coat_l_06}
    componentLinearAccScale=<1 1 1>
    componentLinearVelScale=<1 1 1>
    gravityOverride {z=-980}
    physicsBodyDefinitions [{
      boundBone {boneName=dyn_coat_l_06}
      boxExtents=<15 30 30>
      constraintSetup {angularLimitsMax=<0 0 0>, angularLimitsMin=<0 0 0>, angularTarget=<0 0 0>}
      localJointOffset=<10 0 0>
     }]
    relativeSpaceBone {boneName=root}
   }
   n102 = Anim Dynamics - Bone: dyn_coat_front_l_06(ComponentPose=@n101, Alpha=1, ExternalForce="0,0,0") {
    angularDampingOverride=1
    angularSpringConstant=20000
    bAngularSpring=true
    bOverrideAngularDamping=true
    bOverrideLinearDamping=true
    bUseGravityOverride=true
    gravityScale=0.35
    linearDampingOverride=0.7
    simulationSpace=BoneRelative
    alphaBoolBlend {blendOption=Linear}
    boundBone {boneName=dyn_coat_front_l_06}
    componentLinearAccScale=<1 1 1>
    componentLinearVelScale=<1 1 1>
    gravityOverride {z=-980}
    physicsBodyDefinitions [{
      boundBone {boneName=dyn_coat_front_l_06}
      boxExtents=<15 30 30>
      constraintSetup {angularLimitsMax=<0 0 0>, angularLimitsMin=<0 0 0>, angularTarget=<0 0 0>}
      localJointOffset=<10 0 0>
     }]
    relativeSpaceBone {boneName=root}
   }
   n103 = RigidBody(ComponentPose=@n102, OverrideWorldGravity=Get Gravity Override Cape()) {
    bClampLinearTranslationLimitToRefPose=true
    bOverrideWorldGravity=true
    overridePhysicsAsset=/BRCosmetics/Characters/Player/Male/Medium/Bodies/M_MED_Realm/Meshes/M_MED_Realm_Cape_PhysicsCrouching
    simulationSpace=BaseBoneSpace
    alphaBoolBlend {blendOption=Linear}
    baseBoneRef {boneName=root}
    componentAppliedLinearAccClamp=<3000 3000 3000>
    componentLinearAccScale=<0.05 0.05 0.1>
    componentLinearVelScale=<0.05 0.05 0.1>
    overrideWorldGravity {z=-980}
    simSpaceSettings {
     maxAngularAcceleration=50
     maxAngularVelocity=1
     maxLinearAcceleration=300
     maxLinearVelocity=150
     velocityScaleZ=0
     externalLinearDragV=<0.2 0.2>
    }
   }
   Save cached pose 'crouchPose'(Pose=Component To Local(ComponentPose=@n103)) {cacheName=crouchPose}
   Save cached pose 'preSimCache'() {cacheName=preSimCache}
   n104 = Get BwdMovement() -> (BwdMovement:double)
   n105 = Trail controller - Bone: cape_06_l(Alpha=@n104) {
    bLimitRotation=true
    bUsePlanarLimit=true
    chainLength=6
    alphaScaleBiasClamp {bInterpResult=true, interpSpeedDecreasing=1, interpSpeedIncreasing=1}
    fakeVelocity=<0.1 0.1 0.1>
    rotationLimits [
     {limitMax=<20 0 20>, limitMin=<-20 -30 -20>}
     {limitMax=<30 0 30>, limitMin=<-30 -30 -30>}
     {limitMax=<50 0 50>, limitMin=<-50 -30 -50>}
     {limitMax=<60 0 60>, limitMin=<-60 -30 -60>}
     {limitMax=<70 0 70>, limitMin=<-70 -30 -70>}
     {limitMax=<180 180 180>}
    ]
    rotationOffsets [<0 0 0>, <0 0 0>, <0 0 0>, <0 0 0>, <0 0 0>, <0 0 0>]
    trailBone {boneName=cape_06_l}
    trailRelaxationSpeed {editorCurveData {keys [<at 0 10 RCIM_Linear>,<at 0.4369 7.5752 RCIM_Linear>,<at 1 5 RCIM_Linear>]}}
   }
   n106 = Trail controller - Bone: cape_06_r(ComponentPose=@n105, Alpha=@n104) {
    bLimitRotation=true
    bUsePlanarLimit=true
    chainLength=6
    alphaScaleBiasClamp {bInterpResult=true, interpSpeedDecreasing=1, interpSpeedIncreasing=1}
    fakeVelocity=<0.1 0.1 0.1>
    rotationLimits [
     {limitMax=<20 0 20>, limitMin=<-20 -30 -20>}
     {limitMax=<30 0 30>, limitMin=<-30 -30 -30>}
     {limitMax=<50 0 50>, limitMin=<-50 -30 -50>}
     {limitMax=<60 0 60>, limitMin=<-60 -30 -60>}
     {limitMax=<70 0 70>, limitMin=<-70 -30 -70>}
     {limitMax=<180 180 180>}
    ]
    rotationOffsets [<0 0 0>, <0 0 0>, <0 0 0>, <0 0 0>, <0 0 0>, <0 0 0>]
    trailBone {boneName=cape_06_r}
    trailRelaxationSpeed {editorCurveData {keys [<at 0 10 RCIM_Linear>,<at 0.4738 8.0479 RCIM_Linear>,<at 1 5 RCIM_Linear>]}}
   }
   n107 = Trail controller - Bone: cape_mid_06_l(ComponentPose=@n106, Alpha=@n104) {
    bLimitRotation=true
    bUsePlanarLimit=true
    chainLength=5
    alphaScaleBiasClamp {bInterpResult=true, interpSpeedDecreasing=1, interpSpeedIncreasing=1}
    fakeVelocity=<0.1 0.1 0.1>
    rotationLimits [
     {limitMax=<30 0 30>, limitMin=<-30 -30 -30>}
     {limitMax=<30 0 30>, limitMin=<-30 -30 -30>}
     {limitMax=<40 0 40>, limitMin=<-30 -30 -30>}
     {limitMax=<40 0 40>, limitMin=<-40 -30 -40>}
     {limitMax=<50 0 50>, limitMin=<-50 -30 -50>}
    ]
    rotationOffsets [<0 0 0>, <0 0 0>, <0 0 0>, <0 0 0>, <0 0 0>]
    trailBone {boneName=cape_mid_06_l}
    trailRelaxationSpeed {editorCurveData {keys [<at 0 10 RCIM_Linear>,<at 0.4774 7.9619 RCIM_Linear>,<at 1 5 RCIM_Linear>]}}
   }
   Trail controller - Bone: cape_mid_06_r(ComponentPose=@n107, Alpha=@n104) {
    bLimitRotation=true
    bUsePlanarLimit=true
    chainLength=5
    alphaScaleBiasClamp {bInterpResult=true, interpSpeedDecreasing=1, interpSpeedIncreasing=1}
    fakeVelocity=<0.1 0.1 0.1>
    rotationLimits [
     {limitMax=<30 0 30>, limitMin=<-30 -30 -30>}
     {limitMax=<30 0 30>, limitMin=<-30 -30 -30>}
     {limitMax=<40 0 40>, limitMin=<-30 -30 -30>}
     {limitMax=<40 0 40>, limitMin=<-40 -30 -40>}
     {limitMax=<50 0 50>, limitMin=<-50 -30 -50>}
    ]
    rotationOffsets [<0 0 0>, <0 0 0>, <0 0 0>, <0 0 0>, <0 0 0>]
    trailBone {boneName=cape_mid_06_r}
    trailRelaxationSpeed {editorCurveData {keys [<at 0 10 RCIM_Linear>,<at 0.4848 8.1768 RCIM_Linear>,<at 1 5 RCIM_Linear>]}}
   }
   Get ADS Up Rifle() -> (ADS Up Rifle:double)
  ]}
 EventGraph {nodes [
   Event BlueprintUpdateAnimation()
   > n1 = Sequence!(execute) -> (
    then_0
    then_1
    then_2:Exec
    then_3:Exec
    then_4:Exec
    then_5:Exec
    then_6:Exec
    then_7:Exec
    then_8:Exec
    then_9:Exec
    then_10:Exec
    then_11:Exec
   )
   > Set arm L Rot(arm L Rot:double=GetAlphaFromReferenceTransformDelta(
     SkeletalMeshComponent=Get Body Mesh() -> (Body Mesh:Skeletal Mesh Component Object Reference)
     BoneName=upperarm_l
     BaseName=spine_05
     DeltaComponentType=EulerY
     InputRange=(X=0,Y=100)
     OutputRange=(X=0,Y=100)
     TwistAxis="1,0,0"
    ).RawOutput) -> (Output_Get:double)
   n2 = Gravity Override State Function DEPRECATED(execute=@n1.then_1, CharacterPartAnimInstance=Self-Reference(), GameTimeInSeconds=0, GravityOverrideParams=Get Gravity Override Params())
   > Set Gravity Override Cape(Gravity Override Cape:Vector=@n2) -> (Output_Get:Vector)
   n3 = Sequence!(execute=@n1.then_10) -> (then_0, then_1)
   > Set Body Mesh(Body Mesh:Skeletal Mesh Component Object Reference) -> (Output_Get:Skeletal Mesh Component Object Reference)
   n4 = PlayerMovement(execute=@n3.then_1, AnimInstance=Self-Reference(), XDotProductCoefficient=1, YDotProductCoefficient=1, ZDotProductCoefficient=1)
   n5 = "Absolute (Float)"(A=0)
   n6 = MapRangeClamped(Value=@n4.YMovement, InRangeA=0, InRangeB=1)
   > Set BwdMovement(BwdMovement:double=MapRangeClamped(Value=float + float(A=@n5, B=@n6) {is=title K2Node_CommutativeAssociativeBinaryOperator}, InRangeA=0, InRangeB=1.4)) -> (Output_Get:double)
   n7 = Get Dynamic Material Instance(execute=@n1.then_11, SkeletalMeshComponent=GetPartSkeletalMeshComponent())
   > n8 = Cast To FortPlayerPawn(Object=GetOwningActor())
   > Set As Fort Player Pawn(As Fort Player Pawn:Fort Player Pawn Object Reference=@n8) -> (Output_Get:Fort Player Pawn Object Reference)
   > n9 = Get!(execute:Exec) -> (then:Exec, else:Exec, As Fort Player Pawn:Fort Player Pawn Object Reference)
   n10 = HasMatchingGameplayTag!(self:Gameplay Tag Asset Interface Interface=@n9, TagToCheck:Gameplay Tag Structure=(TagName="Item.Weapon.Galileo.Lobster.Realm")) -> (ReturnValue:bool)
   > n11 = Branch(Condition=@n10)
   > n12 = Branch(Condition=Get bIsSkyDiving())
   SetScalarParameterValue(execute=[@n21.then, @n12.then, @n22.else, @n11.else], self=@n7, ParameterName=Hide Element 02, Value=0)
   SetScalarParameterValue(execute=[@n12.else, @n22.then], self=@n7, ParameterName=Hide Element 02, Value=1)
   n13 = Get bIsParachuteOpened()
   n14 = Branch(execute=@n1.then_2, Condition=AND Boolean(A=Get bIsSkyDiving(), B=NOT Boolean(A=@n13)))
   > Set skydiveNoParachute(skydiveNoParachute:bool=true) -> (Output_Get:bool)
   Set skydiveNoParachute(execute=@n14.else, skydiveNoParachute=false)
   Set Is Wearing Backbling(execute=@n1.then_3, Is Wearing Backbling=CheckForBackbling())
   Set thigh L Rot(
    execute=@n1.then_4
    thigh L Rot:double=GetAlphaFromReferenceTransformDelta(SkeletalMeshComponent=Get Body Mesh(), BoneName=thigh_l, BaseName=pelvis, InputRange=(X=-45,Y=-15), OutputRange=(X=1,Y=0), TwistAxis="1,0,0")
   ) -> (Output_Get:double)
   Set Body Mesh(execute=@n1.then_5, Body Mesh=GetPartSkeletalMeshComponent(self=Self-Reference()))
   n15 = Branch(execute=@n1.then_6, Condition=Get bIsTargeting())
   n16 = BreakTransform(InTransform=GetDeltaTransformFromRefPose(self=GetOwnerSkeletalMeshComponent(), BoneName=clavicle_l, BaseName=spine_05))
   > Set ADS Up Rifle(ADS Up Rifle:double=Remap Float(Input=BreakRotator(InRot=@n16.Rotation).Pitch, OldRange=(X=0,Y=40), NewRange=(X=0,Y=1)) -> (Output=~)) -> (Output_Get:double)
   Set ADS Up Rifle(execute=@n15.else, ADS Up Rifle=0)
   n17 = PlayerMovement(execute=@n1.then_8, AnimInstance=Self-Reference(), XDotProductCoefficient=1, YDotProductCoefficient=1, ZDotProductCoefficient=1)
   > n18 = Branch(Condition=OR Boolean!(A="Not Equal (Float)"(A=@n17), B="Not Equal (Float)"(A=@n17.YMovement), C:bool="Not Equal (Float)"(A=@n17.ZMovement)) -> (ReturnValue))
   > Set charMoving(charMoving:bool=false) -> (Output_Get:bool)
   Set charMoving(execute=@n18.else, charMoving=false)
   n19 = Get charMoving() -> (charMoving:bool)
   n20 = Branch(execute=@n1.then_9, Condition=AND Boolean(A=Get bIsCrouching(), B=NOT Boolean(A=@n19)))
   > Set isCrouchStopped(isCrouchStopped:bool=true) -> (Output_Get:bool)
   Set isCrouchStopped(execute=@n20.else, isCrouchStopped=false)
   n21 = Event OnBeginSkydiving()
   Event OnEndSkydiving()
   > n22 = Branch(Condition=@n10)
   GetPartSkeletalMeshComponent(self=Get Target() -> (Target:Custom Character Part Anim Instance Object Reference))
   Event OnWeaponChanged()
   Branch(Condition=Get bIsSprinting())
  ]}
]
variables [
 ADS Up Rifle {type=number}
 arm L Rot {type=number}
 As Fort Player Pawn {type=FortPlayerPawn}
 Body Mesh {type=SkeletalMeshComponent}
 BwdMovement {type=number}
 Char Moving {type=boolean, default=true}
 Character Part Anim Instance {type=CustomCharacterPartAnimInstance, default=null}
 Character Part Anim Instance_0 {type=CustomCharacterPartAnimInstance, default=null}
 charMoving {type=boolean}
 Gravity Override Cape {type=Vector}
 Gravity Override Params {
  type=GravityOverrideParamsStruct
  default {onGround_Moving {gravityOverride=<0 -500 -750>, windAmplitude=<0 0 300>, windFrequency=<0 -300 300>}}
 }
 Gravity Override Params_0 {type=GravityOverrideParamsStruct}
 Is Wearing Backbling {type=boolean}
 isCrouchStopped {type=boolean}
 skydiveNoParachute {type=boolean}
 Target {type=CustomCharacterPartAnimInstance, default=null}
 thigh L Rot {type=number}
]
